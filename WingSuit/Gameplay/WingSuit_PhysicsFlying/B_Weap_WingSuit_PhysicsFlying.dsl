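; Weapon Blueprint for the physics-flying WingSuit item, handling equip feedback, flight deploy, smoke color, ammo cooldown cues and durability.
; Gameplay events select the smoke color (randomized on authority at BeginPlay), and movement-mode changes with an open parachute request wingsuit deployment instead.
; Ammo delegates drive recharge cues, while input-method changes update the gamepad flag.
; At the durability cutoff it strips spare ammo, and an empty item is removed on unequip.
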(event FortWeapon|EventOnEquippedAndAttached
  ; Equipped & Attached
  (GameplayCue|ExecuteGameplayCueOnActor(Burst) (Variables|Default|GetPlayerPawn) (Variables|Default|GetGCWeaponEquippedTag) (Ability|GameplayCue|MakeGameplayCueParameters 0.0))
  (Variables|Default|SetPreviousLocalMagAmmo (FortWeapon|GetMagazineAmmoCount))
  (EventDispatchers|BindEventtoOnLocalAmmoChanged (EventDispatchers|CreateEvent :!event Custom|LocalAmmoChanged))
  ;;[0.04916 0.1312 0.15 0.5] Play SFX if player attempts to use weapon while on cooldown.
  (Ability|Async|WaitGameplayEventToActor (Variables|Default|GetPlayerPawn) "(TagName=\"Event.Primary.Pressed\")"
    (:then
      (Ability|Async|WaitGameplayEventToActor (Variables|Default|GetPlayerPawn) "(TagName=\"Event.Secondary.Pressed\")"
        (:then
          ; Play equip sound the first time the item is equipped
          (Utilities|FlowControl|DoOnce false
            (:Completed
              (Networking|CanExecuteCosmeticEvents
                (:True
                  (SoundLibrary|PlaySoundSimple (WorldSubsystems|GetSoundLibrarySubsystem) (Game|GetInstigator) (Utilities|Struct|MakeSimpleSoundSettings "(TagName=\"SoundLibrary.Item.Pickup\")" true false true false 1.0 1.0 0.0 0 0 "None" (Variables|Visual|GetWeaponMesh)) :!class /Script/SoundLibrary.SoundLibrarySubsystem))))))
        (:EventReceived
          _)))
    (:EventReceived
      _))
  (if (Math|Boolean|ANDBoolean (== (FortWeapon|GetMagazineAmmoCount) 0) (Pawn|IsLocallyViewed (Variables|Default|GetPlayerPawn)) (FortWeapon|IsEquipped))
    (Audio|PlaySound2D (Variables|Default|GetDryFireSound))
    (bind _asfort_player_pawn (Utilities|Casting|CastToFortPlayerPawn (Variables|Default|GetPlayerPawn)))
    (FailedAction|UnabletoPerformAction _asfort_player_pawn true (Variables|Default|GetFailed%20Reason))))

(event Custom|FlightStarted
  ; Flight Started
  (Ability|GameplayEffect|AddLooseGameplayTags (Variables|Default|GetPlayerPawn) (Variables|Default|GetUnfurlTag))
  (Ability|SendGameplayEventtoActor (Variables|Default|GetPlayerPawn) "(TagName=\"WingSuit.BeginFlight\")")
  (Utilities|Time|SetTimerbyEvent (delegate Custom|EndUnfurl) 0.1))

(event Custom|EndUnfurl
  ; Flight Started
  (Ability|GameplayEffect|RemoveLooseGameplayTags (Variables|Default|GetPlayerPawn) (Variables|Default|GetUnfurlTag) true))

(event Custom|SetSmokeZero
  ; Monitor Smoke Tags
  (Variables|Default|SetSmokeScreenColor 0))

(event Custom|SetSmokeOne
  ; Monitor Smoke Tags
  (Variables|Default|SetSmokeScreenColor 1))

(event Custom|Event_OnConsumeAmmo
  (params (Quantity :type Integer))
  ; On Consume Ammo
  (Variables|Default|SetPreviousLocalMagAmmo (FortWeapon|GetMagazineAmmoCount)))

(event Custom|SetSmokeTwo
  ; Monitor Smoke Tags
  (Variables|Default|SetSmokeScreenColor 2))

(event Custom|SetSmokeThree
  ; Monitor Smoke Tags
  (Variables|Default|SetSmokeScreenColor 3))

(event Custom|SetSmokeFour
  ; Monitor Smoke Tags
  (Variables|Default|SetSmokeScreenColor 4))

(event Custom|SetSmokeFive
  ; Monitor Smoke Tags
  (Variables|Default|SetSmokeScreenColor 5))

(event Custom|SetSmokeSix
  ; Monitor Smoke Tags
  (Variables|Default|SetSmokeScreenColor 6))

(event Custom|LocalAmmoChanged
  (params (LocalCount :type Integer) (LocalRemaining :type Integer))
  ; Ammo Changed
  ;;[0.15 0.0572 0.1304 0.5] Play GC when cooldown complete
  (Variables|Default|SetIsCoolingDown false)
  (if (and (> LocalCount (Variables|Default|GetPreviousLocalMagAmmo)) (FortWeapon|IsEquipped))
    (GameplayCue|ExecuteGameplayCue (Variables|Default|GetPlayerPawn) (Variables|Default|GetGCAmmoRechargedTag))
    (if (Pawn|IsLocallyViewed (Game|GetPlayerPawn 0))
      _)))

(event EventBeginPlay
  (bind _returnvalue (LocalPlayerSubsystems|GetCommonInputSubsystem))
  (Utilities|FlowControl|Sequence
    (:then_0
      ; Monitor Smoke Tags
      (Ability|Async|WaitGameplayEventToActor (Game|GetInstigator) "(TagName=\"WingSuit.SetSmokeColor.Zero\")"
        (:then
          (Ability|Async|WaitGameplayEventToActor (Game|GetInstigator) "(TagName=\"WingSuit.SetSmokeColor.One\")"
            (:then
              (Ability|Async|WaitGameplayEventToActor (Game|GetInstigator) "(TagName=\"WingSuit.SetSmokeColor.Two\")"
                (:then
                  (Ability|Async|WaitGameplayEventToActor (Game|GetInstigator) "(TagName=\"WingSuit.SetSmokeColor.Three\")"
                    (:then
                      (Ability|Async|WaitGameplayEventToActor (Game|GetInstigator) "(TagName=\"WingSuit.SetSmokeColor.Four\")"
                        (:then
                          (Ability|Async|WaitGameplayEventToActor (Game|GetInstigator) "(TagName=\"WingSuit.SetSmokeColor.Five\")"
                            (:then
                              (Ability|Async|WaitGameplayEventToActor (Game|GetInstigator) "(TagName=\"WingSuit.SetSmokeColor.Six\")"
                                (:EventReceived
                                  (CallFunction|SetSmokeSix))))
                            (:EventReceived
                              (CallFunction|SetSmokeFive))))
                        (:EventReceived
                          (CallFunction|SetSmokeFour))))
                    (:EventReceived
                      (CallFunction|SetSmokeThree))))
                (:EventReceived
                  (CallFunction|SetSmokeTwo))))
            (:EventReceived
              (CallFunction|SetSmokeOne))))
        (:EventReceived
          (CallFunction|SetSmokeZero))))
    (:then_1
      ; Randomize Initial Smoke Color
      (Networking|SwitchHasAuthority
        (:Authority
          (Variables|Default|SetSmokeScreenColor (Math|Random|RandomIntegerinRange 0 7)))))
    (:then_2
      ; Begin Play
      (Variables|Default|SetPlayerPawn (Utilities|Casting|CastToFortPlayerPawn (Game|GetInstigator)))
      (EventDispatchers|BindEventtoOnConsumeAmmo (EventDispatchers|CreateEvent :!event Custom|Event_OnConsumeAmmo))
      (bind _outrow (Fortnite|GetWeaponStatsRow (Data|GetWeaponDataWeaponStatHandle)))
      (Variables|Default|SetCachedWeaponBaseDurabilityToUse (Utilities|Struct|BreakFortBaseWeaponStats _outrow))
      (Variables|Default|SetCachedClipSize (.ClipSize (Utilities|Struct|BreakFortBaseWeaponStats _outrow)))
      (bind _playerpawn (Variables|Default|GetPlayerPawn)
        (:then
          (Character|AssignMovementModeChangedDelegate _playerpawn (EventDispatchers|CreateEvent :!event On%20Movment%20Mode%20Changed))
          (if (Pawn|IsLocallyControlled _playerpawn)
            (Utilities|IsValid _returnvalue
              (:Is%20Valid
                (CommonInputSubsystem|BindEventtoOnInputMethodChanged _returnvalue (delegate Custom|HandleInputMethodChanged))
                (CallFunction|HandleInputMethodChanged "Count")))))
        (:else)))))

(event Custom|HandleInputMethodChanged
  (params (bNewInputType :type ECommonInputType%20Enum :default "MouseAndKeyboard"))
  ; Begin Play
  (bind _playerpawn (Variables|Default|GetPlayerPawn)
    (:then
      (bind _asfort_player_controller (Utilities|Casting|CastToFortPlayerController (Pawn|GetController _playerpawn)))
      (CallFunction|OnServerSetIsUsingGamepad (Input|IsUsingGamepad _asfort_player_controller)))
    (:else)))

(event Custom|OnServer_SetIsUsingGamepad
  (params (IsUsingGamepad :type Boolean))
  ; Begin Play
  (Variables|Default|SetIs%20Using%20Gamepad IsUsingGamepad))

(event EventEndPlay
  (params (EndPlayReason :type EEndPlayReason%20Enum :default "Destroyed"))
  (bind _returnvalue (LocalPlayerSubsystems|GetCommonInputSubsystem))
  ; End Play
  (EventDispatchers|UnbindEventfromOnConsumeAmmo (EventDispatchers|CreateEvent :!event Custom|Event_OnConsumeAmmo))
  (bind _playerpawn (Variables|Default|GetPlayerPawn)
    (:then
      (Character|UnbindEventfromMovementModeChangedDelegate _playerpawn (EventDispatchers|CreateEvent :!event On%20Movment%20Mode%20Changed))
      (Utilities|IsValid _returnvalue
        (:Is%20Valid
          (CommonInputSubsystem|UnbindEventfromOnInputMethodChanged _returnvalue (EventDispatchers|CreateEvent :!event Custom|HandleInputMethodChanged)))))
    (:else)))

(event Custom|TriggerCooldownAudio
  ; Trigger and fade out cooldown sound
  (if (Pawn|IsLocallyViewed (Variables|Default|GetPlayerPawn))
    (Variables|Default|GetIsCoolingDown
      (:then)
      (:else
        (bind _returnvalue (Audio|SpawnSoundAttached "/WingSuit/Audio/SC_Wingsuit_Cooldown_Start.SC_Wingsuit_Cooldown_Start" (Variables|Visual|GetWeaponMesh)))
        (Variables|Default|SetCoolDownAudComp _returnvalue)
        (Variables|Default|SetIsCoolingDown true)))))

(event Custom|CheckRemoveWeaponOnEmpty
  ;[1 1 1] Check Remove Weapon On Empty
  (if (<= (FortWeapon|GetLocalRemainingAmmo) 0)
    (Inventory|RemoveItemFromPlayerByGuid (Fortnite|GetFortPlayerControllerfromActor (Game|GetInstigator)) (FortWeapon|GetInventoryGUID) 1 true)))

(event FortWeapon|EventOnUnEquippedAndDetached
  ; Trigger and fade out cooldown sound
  (|CallParentFunction)
  (CallFunction|CheckRemoveWeapononEmpty)
  (bind _cooldownaudcomp (Variables|Default|GetCoolDownAudComp)
    (:then
      (Audio|Components|Audio|FadeOut _cooldownaudcomp 0.25 0.0 "Linear" :!class /Script/Engine.AudioComponent))
    (:else)))

(event Custom|Server_SetFreeBuild
  (params (New%20Value :type Boolean))
  (Networking|SwitchHasAuthority
    (:Authority
      (Variables|Default|SetFreeDeploy New%20Value))))

(event Custom|Client_SetFreeBuild
  (params (New%20Value :type Boolean))
  (Networking|SwitchHasAuthority
    (:Remote
      (bind _output_get (Variables|Default|SetFreeDeploy New%20Value))
      (CallFunction|ServerSetFreeBuild _output_get))))

(fn CheckForDurabilityCutoff
  (params (DurabilityToUse :type Float%20%28double-precision%29))
  (bind (_bisvalid _returnvalue) (FortWeapon|GetDurability))
  (if _bisvalid
    (return (and (<= (- _returnvalue DurabilityToUse) 0) (> (ScalableFloat|GetValueatLevel (Variables|Default|GetbRespectRemainingAmmo)) 0)))
    (else
      (return false))))

(fn GetAnimData
  (return (Variables|Default|GetCharacterPart) (Variables|Default|GetIs%20Using%20Gamepad)))

(fn GetData
  (return (Variables|Default|GetLaunchActivated)))

(fn GetStartingAnim
  (bind _result (ArrayMacros|SafeArrayGet (Variables|Default|GetAnimList) (Variables|Default|GetStartingAnimIncrementer))
    (:%20%20
      (CallFunction|GetStartingAnimIncrementer)
      (return _result))))

(fn GetStartingAnimIncrementer
  (bind _result (Math|Integer|IncrementInt (Variables|Default|GetStartingAnimIncrementer))
    (:%20%20
      (if (> _result 3)
        (Variables|Default|SetStartingAnimIncrementer 1))
      (return (Variables|Default|GetStartingAnimIncrementer)))))

(fn HandleReachedDurabilityCutoff
  (bind (_bisvalid _returnvalue) (FortWeapon|GetDurability))
  ;[1 1 1] When we reach the cutoff, remove all ammo except what's in the magazine plus what's recharging. When we have one remaining shot, and it's in the magazine, destroy the weapon. Set durability to 1% when we reach this stage to guarantee any future shots will run this logic.
  (if _bisvalid
    (LocalVariables|SetCurrentDurability _returnvalue)
    (FortWeapon|UseWeaponDurability(Explicit) (- (LocalVariables|GetCurrentDurability) 1.0))
    (if (and (== (FortWeapon|GetMagazineAmmoCount) 1) (== (FortWeapon|GetRemainingAmmo) 1))
      (FortWeapon|UseWeaponDurability(Explicit) (LocalVariables|GetCurrentDurability))
      (else
        (LocalVariables|SetAmmoToRemove (- (FortWeapon|GetRemainingAmmo) (FortWeapon|GetMagazineAmmoCount)))
        ;[1 1 1] We add 1 because if the magazine is not currently full, we are recharging a pip, so the 1 represents the charging pip
        (if (< (FortWeapon|GetMagazineAmmoCount) (Variables|Default|GetCachedClipSize))
          (LocalVariables|SetAmmoToRemove (- (LocalVariables|GetAmmoToRemove) 1))
          (else
            _))
        (FortWeapon|Ammo|ConsumeSpareAmmo (LocalVariables|GetAmmoToRemove) false)))))

(fn OnMovmentModeChanged
  (params (Character :type Character%20Object%20Reference) (PrevMovementMode :type EMovementMode%20Enum :default "MOVE_None") (PreviousCustomMode :type Byte))
  (if (and (FortWeapon|IsEquipped) (Parachute|IsParachuteOpen (Variables|Default|GetPlayerPawn)))
    (if (not (GameplayTags|HasMatchingGameplayTag (Variables|Default|GetPlayerPawn) "(TagName=\"MovementMode.Flying.WingSuit\")"))
      (CallFunction|ClientSetFreeBuild true)
      (Ability|SendGameplayEventtoActor (Variables|Default|GetPlayerPawn) "(TagName=\"WingSuit.DeployWingSuitInstead\")"))))

(fn PrintDebugDurability
  (bind (_bisvalid _returnvalue) (FortWeapon|GetDurability))
  (Networking|SwitchHasAuthority
    (:Authority
      (if (and (FortWeapon|GetMaxDurability) (FortWeapon|GetDurability))
        (Development|PrintString (Utilities|String|Append "Mag: " (Utilities|String|ToString(Integer) (FortWeapon|GetMagazineAmmoCount))) true false "(R=0.000000,G=0.660000,B=1.000000,A=1.000000)" 2.0 "PersistR")
        (Development|PrintString (Utilities|String|Append "Durability: " (Utilities|String|ToString(Float) _returnvalue)) true false "(R=0.226641,G=1.000000,B=0.000000,A=1.000000)" 2.0 "PersistD")
        (Development|PrintString (Utilities|String|Append "Ammo: " (Utilities|String|ToString(Integer) (FortWeapon|GetRemainingAmmo))) true false "(R=0.001370,G=1.000000,B=0.618148,A=1.000000)" 2.0 "PersistA")))))

(fn TryUseDurability
  (params (AbilityDurabilityMultiplier :type Float%20%28double-precision%29))
  (Networking|SwitchHasAuthority
    (:Authority
      (LocalVariables|SetDurabilityToUse (* AbilityDurabilityMultiplier (Variables|Default|GetCachedWeaponBaseDurabilityToUse)))
      (if (CallFunction|CheckforDurabilityCutoff (LocalVariables|GetDurabilityToUse))
        (CallFunction|HandleReachedDurabilityCutoff)
        (else
          (FortWeapon|UseWeaponDurability(Explicit) (LocalVariables|GetDurabilityToUse)))))))

(fn ConstructionScript)
