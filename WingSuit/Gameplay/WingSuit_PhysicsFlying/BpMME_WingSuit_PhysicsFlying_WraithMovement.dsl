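; Movement mode extension that drives wingsuit physics flight, named WraithMovement and reusing some MistRaven parts.
; Activation and deactivation manage input scaling, camera, trail and speed-based lens VFX, optionally triggering a roll landing.
; Post-movement updates store player or bot input in runtime data; PhysUpdate turns it into rotation and velocity, with pitch controlling speed and slerped drag smoothing direction.
; Hit response selects slide, normal bump or ricochet, retrying up to three passes.
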(event EnhancedInputActionIA_WingSuit_PhysicsFlying_TempTestingAction
  (params (ActionValue :type Boolean) (ElapsedSeconds :type Float%20%28double-precision%29) (TriggeredSeconds :type Float%20%28double-precision%29) (InputAction :type Input%20Action%20Object%20Reference :default "/WingSuit/Gameplay/WingSuit_PhysicsFlying/Input/IA_WingSuit_PhysicsFlying_TempTestingAction.IA_WingSuit_PhysicsFlying_TempTestingAction")))

(event Custom|PushInputContext
  (params (Controller :type Controller%20Object%20Reference))
  ;[1 0 0] Input (WIP non-functioning without backing code changes to MME system)
  (bind _asfort_player_controller (Utilities|Casting|CastToFortPlayerController Controller))
  (if (not (Networking|HasAuthority _asfort_player_controller))
    (Input|AddMappingContext (PlayerController|LocalPlayerSubsystems|GetEnhancedInputLocalPlayerSubsystem _asfort_player_controller) (Variables|Default|GetMistRavenInputContext) 0 "(bIgnoreAllPressedKeysUntilRelease=True,bForceImmediately=False,bNotifyUserSettings=False)")))

(event EnhancedInputActionIA_WingSuit_PhysicsFlyingFlyUp
  (params (ActionValue :type Boolean) (ElapsedSeconds :type Float%20%28double-precision%29) (TriggeredSeconds :type Float%20%28double-precision%29) (InputAction :type Input%20Action%20Object%20Reference :default "/WingSuit/Gameplay/WingSuit_PhysicsFlying/Input/IA_WingSuit_PhysicsFlyingFlyUp.IA_WingSuit_PhysicsFlyingFlyUp")))

(event Custom|PopInputContext
  (params (Controller :type Controller%20Object%20Reference))
  ;[1 0 0] Input (WIP non-functioning without backing code changes to MME system)
  (bind _asfort_player_controller (Utilities|Casting|CastToFortPlayerController Controller))
  (if (not (Networking|HasAuthority _asfort_player_controller))
    (Input|RemoveMappingContext (PlayerController|LocalPlayerSubsystems|GetEnhancedInputLocalPlayerSubsystem _asfort_player_controller) 0 "(bIgnoreAllPressedKeysUntilRelease=True,bForceImmediately=False,bNotifyUserSettings=False)")))

(event EnhancedInputActionIA_WingSuit_PhysicsFlyingFlyDown
  (params (ActionValue :type Boolean) (ElapsedSeconds :type Float%20%28double-precision%29) (TriggeredSeconds :type Float%20%28double-precision%29) (InputAction :type Input%20Action%20Object%20Reference :default "/WingSuit/Gameplay/WingSuit_PhysicsFlying/Input/IA_WingSuit_PhysicsFlyingFlyDown.IA_WingSuit_PhysicsFlyingFlyDown")))

(event Custom|LoopingVfx
  (bind _ownerpawn (Variables|FortMovementMode_BaseExtLogic|GetOwnerPawn))
  (bind _returnvalue (Fortnite|GetFortPlayerControllerfromActor _ownerpawn))
  (bind _playercameramanager (Variables|/Script/Engine.PlayerController|GetPlayerCameraManager _returnvalue))
  (if (> (Math|Vector|VectorLength (Transformation|GetVelocity (Variables|FortMovementMode_BaseExtLogic|GetOwnerPawn))) 1.8e+03)
    ;[1 1 1] Fast
    (if (not (Variables|Default|GetlensActive))
      (Variables|Default|SetlensActive true)
      (bind _returnvalue_2 (Camera|AddGenericCameraLensEffect _playercameramanager "/WingSuit/VFX/BP_CameraLens_WingSuit_Sprint.BP_CameraLens_WingSuit_Sprint_C"))
      (Variables|Default|Setcamlenseffect _returnvalue_2))
    (elif (Variables|Default|GetlensActive)
      ;[1 1 1] Slow
      (Variables|Default|SetlensActive false)
      (Camera|RemoveGenericCameraLensEffect _playercameramanager (Variables|Default|Getcamlenseffect))
      (Components|Activation|Deactivate (CameraLensEffect|GetPrimaryParticleComponent (Variables|Default|Getcamlenseffect))))))

(event EventOnExtensionActivate
  (params (RuntimeContext :type MMERuntime%20Context%20Structure%20%28by%20ref%29))
  (bind _ownerpawn (Variables|FortMovementMode_BaseExtLogic|GetOwnerPawn))
  (bind _returnvalue (Actor|GetComponentbyClass _ownerpawn "/MistRaven/B_MistRaven_PlayerComponent.B_MistRaven_PlayerComponent_C"))
  (bind _returnvalue_2 (Class|FortMovementModeExtBlueprintLibrary|GetRuntimeDatafromContext RuntimeContext "/WingSuit/Gameplay/WingSuit_PhysicsFlying/BpMME_RuntimeData_WingSuit_PhysicsFlying_WraithMovement.BpMME_RuntimeData_WingSuit_PhysicsFlying_WraithMovement_C"))
  (Variables|/WingSuit/Gameplay/WingSuit_PhysicsFlying/BpMME_RuntimeData_WingSuit_PhysicsFlying_WraithMovement.BpMME_RuntimeData_WingSuit_PhysicsFlying_WraithMovement_C|SetPawnForward (Transformation|GetActorForwardVector (Variables|FortMovementMode_BaseExtLogic|GetOwnerPawn)) _returnvalue_2)
  (bind _asfort_player_controller (Utilities|Casting|CastToFortPlayerController (Pawn|GetController (Variables|FortMovementMode_BaseExtLogic|GetOwnerPawn))))
  (if (Utilities|Enum|Equal(Enum) (Networking|GetLocalRole (Variables|FortMovementMode_BaseExtLogic|GetOwnerPawn)) "ROLE_AutonomousProxy")
    (bind _output_get (Variables|/WingSuit/Gameplay/WingSuit_PhysicsFlying/BpMME_RuntimeData_WingSuit_PhysicsFlying_WraithMovement.BpMME_RuntimeData_WingSuit_PhysicsFlying_WraithMovement_C|SetIsOnGamepad (Input|IsUsingGamepad _asfort_player_controller) _returnvalue_2))
    (if _output_get
      ;[1 1 1] TODO: also change if the player changes controller type mid flight
      (Input|SetInputPitchScale _asfort_player_controller (ScalableFloat|GetValueatLevel (Variables|Default|GetRow_TurnSpeedPitch_Gamepad)))
      (Input|SetInputYawScale _asfort_player_controller (ScalableFloat|GetValueatLevel (Variables|Default|GetRow_TurnSpeedYaw_Gamepad)))
      (else
        (Input|SetInputPitchScale _asfort_player_controller (ScalableFloat|GetValueatLevel (Variables|Default|GetRow_TurnSpeedPitch_MouseKeyboard)))
        (Input|SetInputYawScale _asfort_player_controller (ScalableFloat|GetValueatLevel (Variables|Default|GetRow_TurnSpeedYaw_MouseKeyboard))))))
  ;[0.2541 1 0.2188] Extension Activate
  (bind _returnvalue_3 (Utilities|Time|SetTimerbyEvent (delegate Custom|LoopingVfx) 0.5 true true))
  (Variables|Default|SetVisualsTimer _returnvalue_3)
  (Utilities|FlowControl|Sequence
    (:then_0
      (bind _returnvalue_4 (Niagara|SpawnSystemAttached "/WingSuit/VFX/NS_WingSuit_Tail.NS_WingSuit_Tail" (Variables|/Script/Engine.Character|GetMesh (Variables|FortMovementMode_BaseExtLogic|GetOwnerPawn)) "root"))
      (Variables|Default|SetCurrentVfx _returnvalue_4))
    (:then_1
      (bind _returnvalue_5 (Game|Player|AddNewCameraModifier (Variables|/Script/Engine.PlayerController|GetPlayerCameraManager (Fortnite|GetFortPlayerControllerfromActor (Variables|FortMovementMode_BaseExtLogic|GetOwnerPawn))) (Variables|Default|GetCamera%20Modifier%20Class)))
      (Variables|Default|SetCameraModifierVar _returnvalue_5)
      (Variables|FortMovementMode_BaseExtLogic|SetCameraMode "/WingSuit/Gameplay/WingSuit_PhysicsFlying/Camera/CameraMode_WingSuit_PhysicsFlying.CameraMode_WingSuit_PhysicsFlying_C")
      (Variables|Default|SetServerTotalTime 0.0) ; Reset if we start a MME.\nWe will likely need to make this\nmore robust, either on the MME,\nplayer component, or something else\nWe don't want to allow player spoofing time.\nBut we may need to handle coming back into\nwraith form 
      (Utilities|FlowControl|Sequence
        (:then_1
          (CallFunction|PushInputContext (Pawn|GetController (Variables|FortMovementMode_BaseExtLogic|GetOwnerPawn))))
        (:then_2
          ;[1 1 1] Handle no input held after dash, set velocity to something reasonable
          (if (< (Math|Vector2D|Vector2DLengthSquared (Math|Conversions|ToVector2D(Vector) (Transformation|GetVelocity (Variables|FortMovementMode_BaseExtLogic|GetOwnerPawn)))) 1.0)
            (Utilities|IsValid _returnvalue
              (:Is%20Valid
                (Variables|/Script/Engine.MovementComponent|SetVelocity (* (Math|Vector|Normalize (Math|Vector|MakeVector (.x (- (Transformation|GetActorLocation (Variables|FortMovementMode_BaseExtLogic|GetOwnerPawn)) (Variables|/MistRaven/B_MistRaven_PlayerComponent.B_MistRaven_PlayerComponent_C|GetStartGrimStrutLocation _returnvalue))) (.y (- (Transformation|GetActorLocation (Variables|FortMovementMode_BaseExtLogic|GetOwnerPawn)) (Variables|/MistRaven/B_MistRaven_PlayerComponent.B_MistRaven_PlayerComponent_C|GetStartGrimStrutLocation _returnvalue))) (Math|Float|Max(Float) (.z (- (Transformation|GetActorLocation (Variables|FortMovementMode_BaseExtLogic|GetOwnerPawn)) (Variables|/MistRaven/B_MistRaven_PlayerComponent.B_MistRaven_PlayerComponent_C|GetStartGrimStrutLocation _returnvalue)))))) (ScalableFloat|GetValueasDouble (Variables|Default|GetFastWraithFlightSpeed))) (Variables|/Script/Engine.Character|GetCharacterMovement (Variables|FortMovementMode_BaseExtLogic|GetOwnerPawn)))))))))))

(event EventOnExtensionEnd
  (params (RuntimeContext :type MMERuntime%20Context%20Structure%20%28by%20ref%29)))

(event EventOnExtensionDeactivate
  (params (RuntimeContext :type MMERuntime%20Context%20Structure%20%28by%20ref%29))
  ;[1 0.1562 0.1562] Extension Deactivate
  (bind _asfort_player_controller_athena (Utilities|Casting|CastToFortPlayerControllerAthena (Pawn|GetController (Variables|FortMovementMode_BaseExtLogic|GetOwnerPawn))))
  (Input|ResetInputRotationScale _asfort_player_controller_athena)
  (Game|Player|RemoveCameraModifier (Variables|/Script/Engine.PlayerController|GetPlayerCameraManager (Fortnite|GetFortPlayerControllerfromActor (Variables|FortMovementMode_BaseExtLogic|GetOwnerPawn))) (Variables|Default|GetCameraModifierVar))
  (CallFunction|PopInputContext (Pawn|GetController (Variables|FortMovementMode_BaseExtLogic|GetOwnerPawn)))
  (Components|Activation|Deactivate (Variables|Default|GetCurrentVfx))
  (if (Variables|Default|GetRollLanding)
    (Class|FortMovementModeExtBlueprintLibrary|ActivateRegisteredMovementModeExtension "/WingSuit/Gameplay/WingSuit_PhysicsFlying/FMM_ExtWingSuitLanding.FMM_ExtWingSuitLanding_C" (Variables|FortMovementMode_BaseExtLogic|GetOwnerPawn)))
  (Utilities|Time|ClearandInvalidateTimerbyHandle (Variables|FortMovementMode_BaseExtLogic|GetOwnerPawn) (Variables|Default|GetVisualsTimer))
  ;[1 1 1] Slow
  (Variables|Default|SetlensActive false)
  (Camera|RemoveGenericCameraLensEffect (Variables|/Script/Engine.PlayerController|GetPlayerCameraManager (Fortnite|GetFortPlayerControllerfromActor (Variables|FortMovementMode_BaseExtLogic|GetOwnerPawn))) (Variables|Default|Getcamlenseffect))
  (Components|Activation|Deactivate (CameraLensEffect|GetPrimaryParticleComponent (Variables|Default|Getcamlenseffect))))

(event EventUpdateAfterCharacterMovement
  (params (RuntimeContext :type MMERuntime%20Context%20Structure%20%28by%20ref%29) (OldLocation :type Vector%20%28by%20ref%29 :default "0, 0, 0") (OldVelocity :type Vector%20%28by%20ref%29 :default "0, 0, 0"))
  (bind _ownerpawn (Variables|FortMovementMode_BaseExtLogic|GetOwnerPawn))
  (bind _returnvalue (Fortnite|GetFortPlayerControllerfromActor _ownerpawn))
  (bind _ownerpawn_2 (Variables|FortMovementMode_BaseExtLogic|GetOwnerPawn))
  ;[0.2396 0.8092 1] UpdateAfterCharacterMovement
  (|CallParentFunction RuntimeContext OldLocation OldVelocity)
  (bind _returnvalue_2 (Class|FortMovementModeExtBlueprintLibrary|GetRuntimeDatafromContext RuntimeContext "/WingSuit/Gameplay/WingSuit_PhysicsFlying/BpMME_RuntimeData_WingSuit_PhysicsFlying_WraithMovement.BpMME_RuntimeData_WingSuit_PhysicsFlying_WraithMovement_C"))
  (Utilities|IsValid _returnvalue_2
    (:Is%20Valid
      (Utilities|FlowControl|Sequence
        (:then_0
          (Variables|/WingSuit/Gameplay/WingSuit_PhysicsFlying/BpMME_RuntimeData_WingSuit_PhysicsFlying_WraithMovement.BpMME_RuntimeData_WingSuit_PhysicsFlying_WraithMovement_C|SetControlRotation (Pawn|GetControlRotation _ownerpawn_2 :!class /Script/Engine.Pawn) _returnvalue_2)
          (Variables|/WingSuit/Gameplay/WingSuit_PhysicsFlying/BpMME_RuntimeData_WingSuit_PhysicsFlying_WraithMovement.BpMME_RuntimeData_WingSuit_PhysicsFlying_WraithMovement_C|SetPawnForward (Transformation|GetActorForwardVector _ownerpawn_2) _returnvalue_2)
          (if (Pawn|IsBotControlled (Variables|FortMovementMode_BaseExtLogic|GetOwnerPawn))
            (Variables|/WingSuit/Gameplay/WingSuit_PhysicsFlying/BpMME_RuntimeData_WingSuit_PhysicsFlying_WraithMovement.BpMME_RuntimeData_WingSuit_PhysicsFlying_WraithMovement_C|SetRequestedVelocity (Pawn|Components|CharacterMovement|GetLastUpdateRequestedVelocity (Variables|/Script/Engine.Character|GetCharacterMovement (Variables|FortMovementMode_BaseExtLogic|GetOwnerPawn))) _returnvalue_2)
            (else
              (Variables|/WingSuit/Gameplay/WingSuit_PhysicsFlying/BpMME_RuntimeData_WingSuit_PhysicsFlying_WraithMovement.BpMME_RuntimeData_WingSuit_PhysicsFlying_WraithMovement_C|SetInputVector (Class|FortMovementBlueprintLibrary|GetLocalMoveInputVector (Variables|FortMovementMode_BaseExtLogic|GetOwnerPawn)) _returnvalue_2)
              (Variables|/WingSuit/Gameplay/WingSuit_PhysicsFlying/BpMME_RuntimeData_WingSuit_PhysicsFlying_WraithMovement.BpMME_RuntimeData_WingSuit_PhysicsFlying_WraithMovement_C|SetUpInput (Class|FortMovementBlueprintLibrary|GetMMELocalLegacyJumpInputPressed _returnvalue) _returnvalue_2)
              (Variables|/WingSuit/Gameplay/WingSuit_PhysicsFlying/BpMME_RuntimeData_WingSuit_PhysicsFlying_WraithMovement.BpMME_RuntimeData_WingSuit_PhysicsFlying_WraithMovement_C|SetDownInput (Class|FortMovementBlueprintLibrary|GetMMELocalLegacyCrouchInputPressed _returnvalue) _returnvalue_2)))
          (Variables|/WingSuit/Gameplay/WingSuit_PhysicsFlying/BpMME_RuntimeData_WingSuit_PhysicsFlying_WraithMovement.BpMME_RuntimeData_WingSuit_PhysicsFlying_WraithMovement_C|SetTotalTimeUpdated (+ (Class|FortMovementModeExtBlueprintLibrary|GetDeltaTimefromContext RuntimeContext (Variables|FortMovementMode_BaseExtLogic|GetbUseDilatedDeltaTime)) (Variables|/WingSuit/Gameplay/WingSuit_PhysicsFlying/BpMME_RuntimeData_WingSuit_PhysicsFlying_WraithMovement.BpMME_RuntimeData_WingSuit_PhysicsFlying_WraithMovement_C|GetTotalTimeUpdated _returnvalue_2)) _returnvalue_2))
        (:then_1)))))

(event Custom|BumpUpSlide
  (params (MMERuntimeContext :type MMERuntime%20Context%20Structure))
  ;[0.4176 0.1208 0.5156] Handle Hit
  (Transformation|SetActorLocation (Variables|FortMovementMode_BaseExtLogic|GetOwnerPawn) (+ (Transformation|GetActorLocation (Variables|FortMovementMode_BaseExtLogic|GetOwnerPawn)) (Math|Vector|MakeVector 0.0 0.0 5.0)) true false) ; sweep so we get collision.\ntry to slide up if we can't slide in any direction\npreferably this would be the step up logic
)

(event EventHandleHitResponse
  (params (RuntimeContext :type MMERuntime%20Context%20Structure%20%28by%20ref%29) (InVelocity :type Vector%20%28by%20ref%29 :default "0, 0, 0") (OldPosition :type Vector%20%28by%20ref%29 :default "0, 0, 0") (NewLocation :type Vector%20%28by%20ref%29 :default "0, 0, 0") (Adjustement :type Vector%20%28by%20ref%29 :default "0, 0, 0") (Hit :type Hit%20Result%20Structure%20%28by%20ref%29))
  ;[0.4176 0.1208 0.5156] Handle Hit
  (switch int (ScalableFloat|GetValueasInteger (Variables|Default|GetHitStrategy))
    (:0
      (bind (_handledblockinghit _outvelocity _outoldposition _outnewposition _outnewadjustment _sweephitresult) (Default|HandleHitSlide InVelocity OldPosition NewLocation Adjustement Hit))
      (if _handledblockinghit
        (bind (_handledblockinghit_2 _outvelocity_2 _outoldposition_2 _outnewposition_2 _outnewadjustment_2 _sweephitresult_2) (Default|HandleHitSlide _outvelocity _outoldposition _outnewposition _outnewadjustment _sweephitresult 1))
        (if _handledblockinghit_2
          (bind _handledblockinghit_3 (Default|HandleHitSlide _outvelocity_2 _outoldposition_2 _outnewposition_2 _outnewadjustment_2 _sweephitresult_2 2))
          (if _handledblockinghit_3
            (CallFunction|BumpUpSlide RuntimeContext)))))
    (:1
      (bind (_handledblockinghit_4 _outvelocity_3 _outoldposition_3 _outnewposition_3 _outnewadjustment_3 _sweephitresult_3) (CallFunction|HandleHitNormalBump InVelocity OldPosition NewLocation Adjustement Hit))
      (if _handledblockinghit_4
        (bind (_handledblockinghit_5 _outvelocity_4 _outoldposition_4 _outnewposition_4 _outnewadjustment_4 _sweephitresult_4) (CallFunction|HandleHitNormalBump _outvelocity_3 _outoldposition_3 _outnewposition_3 _outnewadjustment_3 _sweephitresult_3 1))
        (if _handledblockinghit_5
          (CallFunction|HandleHitNormalBump _outvelocity_4 _outoldposition_4 _outnewposition_4 _outnewadjustment_4 _sweephitresult_4 2))))
    (:2
      (bind (_handledblockinghit_6 _outvelocity_5 _outoldposition_5 _outnewposition_5 _outnewadjustment_5 _sweephitresult_5) (CallFunction|HandleHitRicochetVelocity InVelocity OldPosition NewLocation Adjustement Hit))
      (if _handledblockinghit_6
        (bind (_handledblockinghit_7 _outvelocity_6 _outoldposition_6 _outnewposition_6 _outnewadjustment_6 _sweephitresult_6) (CallFunction|HandleHitRicochetVelocity _outvelocity_5 _outoldposition_5 _outnewposition_5 _outnewadjustment_5 _sweephitresult_5 1))
        (if _handledblockinghit_7
          (bind _handledblockinghit_8 (CallFunction|HandleHitRicochetVelocity _outvelocity_6 _outoldposition_6 _outnewposition_6 _outnewadjustment_6 _sweephitresult_6 2))
          (if _handledblockinghit_8
            _))))))

(event EventPhysicsRotation
  (params (RuntimeContext :type MMERuntime%20Context%20Structure%20%28by%20ref%29)))

(fn PhysUpdate
  (params (RuntimeContext :type MMERuntime%20Context%20Structure%20%28by%20ref%29) (InVelocity :type Vector%20%28by%20ref%29 :default "0, 0, 0") (InAcceleration :type Vector%20%28by%20ref%29 :default "0, 0, 0"))
  (bind _newrotation (LocalVariables|GetnewRotation))
  (bind _returnvalue (Math|Vector|GetForwardVector _newrotation))
  (bind _z (.z _returnvalue))
  (bind _controlrotation_rotatedbyinput (LocalVariables|GetControlRotation_RotatedByInput))
  (bind _pitch (.pitch _controlrotation_rotatedbyinput))
  (bind _newvelocity (LocalVariables|GetNewVelocity))
  (bind _returnvalue_2 (Math|Vector|VectorLengthXY _newvelocity))
  (bind _returnvalue_3 (Math|Vector|VectorUp))
  (bind _runtimetarget (LocalVariables|GetRuntimeTarget))
  (bind _pawnforward (Variables|/WingSuit/Gameplay/WingSuit_PhysicsFlying/BpMME_RuntimeData_WingSuit_PhysicsFlying_WraithMovement.BpMME_RuntimeData_WingSuit_PhysicsFlying_WraithMovement_C|GetPawnForward _runtimetarget))
  (bind _returnvalue_4 (Math|Rotator|MakeRotfromXZ _pawnforward _returnvalue_3))
  (bind _pitch_2 (.pitch _returnvalue_4))
  (bind (_returnvalue_roll _returnvalue_pitch _returnvalue_yaw) (Math|Rotator|BreakRotator (Math|Rotator|Lerp(Rotator) (Math|Rotator|MakeRotator 0.0 0.0 (.yaw _returnvalue_4)) (Math|Rotator|MakeRotator 0.0 0.0 (.yaw _controlrotation_rotatedbyinput)) (* (Class|FortMovementModeExtBlueprintLibrary|GetDeltaTimefromContext (Variables|GetRuntimeContext) (Variables|FortMovementMode_BaseExtLogic|GetbUseDilatedDeltaTime)) (Variables|Default|GetRow_CharacterFacingTowardsCameraYaw)) true)))
  (bind (_returnvalue_roll_2 _returnvalue_pitch_2) (Math|Rotator|BreakRotator (Math|Rotator|Lerp(Rotator) (Math|Rotator|MakeRotator 0.0 _pitch_2) (Math|Rotator|MakeRotator 0.0 _pitch) (* (Class|FortMovementModeExtBlueprintLibrary|GetDeltaTimefromContext (Variables|GetRuntimeContext) (Variables|FortMovementMode_BaseExtLogic|GetbUseDilatedDeltaTime)) (LocalVariables|GetPitchMultiplier)) true)))
  (bind _busedilateddeltatime (Variables|FortMovementMode_BaseExtLogic|GetbUseDilatedDeltaTime))
  (bind _runtimecontext (Variables|GetRuntimeContext))
  (bind _returnvalue_5 (Class|FortMovementModeExtBlueprintLibrary|GetDeltaTimefromContext _runtimecontext _busedilateddeltatime))
  (bind _invelocity (Variables|GetInVelocity))
  (bind _returnvalue_6 (Math|Vector|VectorLength _invelocity))
  (bind _invelocity_2 (Variables|GetInVelocity))
  (bind _returnvalue_7 (Math|Vector|VectorLength _invelocity_2))
  (bind _invelocity_3 (Variables|GetInVelocity))
  (bind _returnvalue_8 (Math|Vector|VectorLength _invelocity_3))
  (bind _stopglidingmultiplier (Variables|Default|GetStopGlidingMultiplier))
  (bind _returnvalue_9 (* _returnvalue_5 _stopglidingmultiplier))
  (bind _returnvalue_10 (Math|Float|Lerp _returnvalue_8 1e+03 _returnvalue_9))
  (bind (_returnvalue_x _returnvalue_y) (Math|Vector|BreakVector (* (Math|Vector|MakeVector (.x (LocalVariables|GetNewVelocity)) (.y (LocalVariables|GetNewVelocity))) (Math|Float|Clamp(Float) (- 1.0 (* (Math|Curves|GetFloatValue (Variables|Default|GetHorizontalFloatCurve) (Math|Vector|VectorLengthXY (LocalVariables|GetNewVelocity))) (Class|FortMovementModeExtBlueprintLibrary|GetDeltaTimefromContext (Variables|GetRuntimeContext) (Variables|FortMovementMode_BaseExtLogic|GetbUseDilatedDeltaTime))))))))
  (Utilities|FlowControl|Sequence
    (:then_0
      (LocalVariables|Setlocalplayercamera (Variables|/Script/Engine.PlayerController|GetPlayerCameraManager (Fortnite|GetFortPlayerControllerfromActor (Variables|FortMovementMode_BaseExtLogic|GetOwnerPawn)))))
    (:then_1
      (bind _returnvalue_11 (Class|FortMovementModeExtBlueprintLibrary|GetRuntimeDatafromContext (Variables|GetRuntimeContext) "/WingSuit/Gameplay/WingSuit_PhysicsFlying/BpMME_RuntimeData_WingSuit_PhysicsFlying_WraithMovement.BpMME_RuntimeData_WingSuit_PhysicsFlying_WraithMovement_C"))
      (LocalVariables|SetRuntimeTarget _returnvalue_11)
      ;[1 1 1] prototype - restrict sideways movement to half and always move forward
      (if (Variables|/WingSuit/Gameplay/WingSuit_PhysicsFlying/BpMME_RuntimeData_WingSuit_PhysicsFlying_WraithMovement.BpMME_RuntimeData_WingSuit_PhysicsFlying_WraithMovement_C|GetIsOnGamepad _returnvalue_11)
        (if (ScalableFloat|GetValueasBool (Variables|Default|GetRow_GamepadLeftAnalogStickFlight))
          (Development|PrintString 0 (Utilities|String|ToString(Vector2d) (Variables|/WingSuit/Gameplay/WingSuit_PhysicsFlying/BpMME_RuntimeData_WingSuit_PhysicsFlying_WraithMovement.BpMME_RuntimeData_WingSuit_PhysicsFlying_WraithMovement_C|GetInputVector _returnvalue_11)))
          (Variables|/WingSuit/Gameplay/WingSuit_PhysicsFlying/BpMME_RuntimeData_WingSuit_PhysicsFlying_WraithMovement.BpMME_RuntimeData_WingSuit_PhysicsFlying_WraithMovement_C|SetInputVector (Math|Vector2D|MakeVector2D (Math|Vector2D|BreakVector2D (Variables|/WingSuit/Gameplay/WingSuit_PhysicsFlying/BpMME_RuntimeData_WingSuit_PhysicsFlying_WraithMovement.BpMME_RuntimeData_WingSuit_PhysicsFlying_WraithMovement_C|GetInputVector _returnvalue_11)) (Math|Float|Clamp(Float) (.Y (Math|Vector2D|BreakVector2D (Variables|/WingSuit/Gameplay/WingSuit_PhysicsFlying/BpMME_RuntimeData_WingSuit_PhysicsFlying_WraithMovement.BpMME_RuntimeData_WingSuit_PhysicsFlying_WraithMovement_C|GetInputVector _returnvalue_11))) -0.5 0.5)) _returnvalue_11)
          (else
            (Variables|/WingSuit/Gameplay/WingSuit_PhysicsFlying/BpMME_RuntimeData_WingSuit_PhysicsFlying_WraithMovement.BpMME_RuntimeData_WingSuit_PhysicsFlying_WraithMovement_C|SetInputVector (Math|Vector2D|MakeVector2D 1.0 (Math|Float|Clamp(Float) (.Y (Math|Vector2D|BreakVector2D (Variables|/WingSuit/Gameplay/WingSuit_PhysicsFlying/BpMME_RuntimeData_WingSuit_PhysicsFlying_WraithMovement.BpMME_RuntimeData_WingSuit_PhysicsFlying_WraithMovement_C|GetInputVector _returnvalue_11))) -0.5 0.5)) _returnvalue_11 :!label _label_5)))
        (Utilities|FlowControl|Sequence
          (:then_0
            (LocalVariables|SetInput2DPressedThisFrame (> (Math|Vector2D|Vector2DLengthSquared (Variables|/WingSuit/Gameplay/WingSuit_PhysicsFlying/BpMME_RuntimeData_WingSuit_PhysicsFlying_WraithMovement.BpMME_RuntimeData_WingSuit_PhysicsFlying_WraithMovement_C|GetInputVector _returnvalue_11)) 0.01))
            (LocalVariables|SetInputVerticalPressedThisFrame (!= (+ (Math|Conversions|ToInteger(Boolean) (Variables|/WingSuit/Gameplay/WingSuit_PhysicsFlying/BpMME_RuntimeData_WingSuit_PhysicsFlying_WraithMovement.BpMME_RuntimeData_WingSuit_PhysicsFlying_WraithMovement_C|GetUpInput _returnvalue_11)) 0) 0))
            (LocalVariables|SetInputPressedThisFrame (Math|Boolean|ORBoolean (LocalVariables|GetInput2DPressedThisFrame) (LocalVariables|GetInputVerticalPressedThisFrame) (and (Pawn|IsBotControlled (Variables|FortMovementMode_BaseExtLogic|GetOwnerPawn)) (> (Math|Vector|VectorLengthSquared (Variables|/WingSuit/Gameplay/WingSuit_PhysicsFlying/BpMME_RuntimeData_WingSuit_PhysicsFlying_WraithMovement.BpMME_RuntimeData_WingSuit_PhysicsFlying_WraithMovement_C|GetRequestedVelocity _returnvalue_11)) 0))))
            (if (Variables|/WingSuit/Gameplay/WingSuit_PhysicsFlying/BpMME_RuntimeData_WingSuit_PhysicsFlying_WraithMovement.BpMME_RuntimeData_WingSuit_PhysicsFlying_WraithMovement_C|GetIsOnGamepad _returnvalue_11)
              (if (ScalableFloat|GetValueasBool (Variables|Default|GetRow_GamepadLeftAnalogStickFlight))
                ;[0.1146 0.1146 0.1146] Decelerate Version (snappy turning)
                (LocalVariables|SetControlRotation_RotatedByInput (Math|Rotator|MakeRotator (.roll (Math|Rotator|MakeRotfromXZ (Variables|/WingSuit/Gameplay/WingSuit_PhysicsFlying/BpMME_RuntimeData_WingSuit_PhysicsFlying_WraithMovement.BpMME_RuntimeData_WingSuit_PhysicsFlying_WraithMovement_C|GetPawnForward (LocalVariables|GetRuntimeTarget)) (Math|Vector|VectorUp))) (Math|Float|Clamp(Float) (+ (Math|Float|SelectFloat (Math|Float|MapRangeClamped (Math|Vector2D|BreakVector2D (Variables|/WingSuit/Gameplay/WingSuit_PhysicsFlying/BpMME_RuntimeData_WingSuit_PhysicsFlying_WraithMovement.BpMME_RuntimeData_WingSuit_PhysicsFlying_WraithMovement_C|GetInputVector _returnvalue_11)) 1.0 0.0 -3e+01) (Math|Float|MapRangeClamped (Math|Vector2D|BreakVector2D (Variables|/WingSuit/Gameplay/WingSuit_PhysicsFlying/BpMME_RuntimeData_WingSuit_PhysicsFlying_WraithMovement.BpMME_RuntimeData_WingSuit_PhysicsFlying_WraithMovement_C|GetInputVector _returnvalue_11)) 0.0 -1.0 0.0 15.0) (> (Math|Vector2D|BreakVector2D (Variables|/WingSuit/Gameplay/WingSuit_PhysicsFlying/BpMME_RuntimeData_WingSuit_PhysicsFlying_WraithMovement.BpMME_RuntimeData_WingSuit_PhysicsFlying_WraithMovement_C|GetInputVector _returnvalue_11)) 0)) (.pitch (Math|Rotator|MakeRotfromXZ (Variables|/WingSuit/Gameplay/WingSuit_PhysicsFlying/BpMME_RuntimeData_WingSuit_PhysicsFlying_WraithMovement.BpMME_RuntimeData_WingSuit_PhysicsFlying_WraithMovement_C|GetPawnForward (LocalVariables|GetRuntimeTarget)) (Math|Vector|VectorUp)))) (ScalableFloat|GetValueasDouble (Variables|Default|GetRow_MinPitch)) (ScalableFloat|GetValueasDouble (Variables|Default|GetRow_MaxPitch))) (+ (Math|Float|MapRangeClamped (.Y (Math|Vector2D|BreakVector2D (Variables|/WingSuit/Gameplay/WingSuit_PhysicsFlying/BpMME_RuntimeData_WingSuit_PhysicsFlying_WraithMovement.BpMME_RuntimeData_WingSuit_PhysicsFlying_WraithMovement_C|GetInputVector _returnvalue_11))) 0.5 -0.5 7e+01 -7e+01) (.yaw (Math|Rotator|MakeRotfromXZ (Variables|/WingSuit/Gameplay/WingSuit_PhysicsFlying/BpMME_RuntimeData_WingSuit_PhysicsFlying_WraithMovement.BpMME_RuntimeData_WingSuit_PhysicsFlying_WraithMovement_C|GetPawnForward (LocalVariables|GetRuntimeTarget)) (Math|Vector|VectorUp))))))
                ;;[1 1 1] Kinematic Update Speed
                (if (< (Math|Vector2D|BreakVector2D (Variables|/WingSuit/Gameplay/WingSuit_PhysicsFlying/BpMME_RuntimeData_WingSuit_PhysicsFlying_WraithMovement.BpMME_RuntimeData_WingSuit_PhysicsFlying_WraithMovement_C|GetInputVector _returnvalue_11)) 0)
                  (LocalVariables|SetPitchMultiplier (ScalableFloat|GetValueasDouble (Variables|Default|GetRow_CharacterFacingTowardsCameraPitchUp)) :!label _label_2)
                  (else
                    (LocalVariables|SetPitchMultiplier (ScalableFloat|GetValueasDouble (Variables|Default|GetRow_CharacterFacingTowardsCameraPitchDown)) :!label _label)))
                (bind _output_get (LocalVariables|SetnewRotation (Math|Rotator|MakeRotator (.roll _returnvalue_4) _returnvalue_pitch_2 _returnvalue_yaw)))
                (Transformation|SetActorRotation (Variables|FortMovementMode_BaseExtLogic|GetOwnerPawn) _output_get false)
                (if (and (< _z 0.2) (> _z -0.2))
                  ;;;[0.1695 0.11 1 0.5] Glide - we are somewhere near the middle
                  (Variables|Default|SetPreviousVelocityLength _returnvalue_10)
                  (LocalVariables|SetNewVelocityLength_Z _returnvalue_10)
                  (LocalVariables|SetNewVelocityLength_Z (Math|Float|Min(Float) (LocalVariables|GetNewVelocityLength_Z) 4e+03))
                  (Variables|Default|SetStopGlidingMultiplier (Math|Float|Max(Float) (Math|Float|Min(Float) (+ (Variables|Default|GetStopGlidingMultiplier) (* (LocalVariables|GetStopGlidingMultiplierDelta) (Class|FortMovementModeExtBlueprintLibrary|GetDeltaTimefromContext (Variables|GetRuntimeContext) (Variables|FortMovementMode_BaseExtLogic|GetbUseDilatedDeltaTime)))) (LocalVariables|GetStopGlidingMultiplierMax)) (LocalVariables|GetStopGlidngMultiplierMin)))
                  (elif (< _z 0)
                    ;;;[0.4619 1 0.3906] Accelerate
                    (Variables|Default|SetPreviousVelocityLength _returnvalue_7)
                    (LocalVariables|SetNewVelocityLength_Z (+ _returnvalue_7 (Utilities|Operators|Multiply (Variables|Default|GetRow_AccelerationRate) (Math|Float|Lerp 0.0 1.0 (Math|Float|Absolute(Float) _z)) (Class|FortMovementModeExtBlueprintLibrary|GetDeltaTimefromContext (Variables|GetRuntimeContext) (Variables|FortMovementMode_BaseExtLogic|GetbUseDilatedDeltaTime)))))
                    (LocalVariables|SetNewVelocityLength_Z (Math|Float|Min(Float) (LocalVariables|GetNewVelocityLength_Z) 4e+03))
                    (else
                      ;;;[1 0.401 0.401 0.5] Deccelerate (7_15 actually accelerate much slower now)
                      (Variables|Default|SetPreviousVelocityLength _returnvalue_6)
                      (LocalVariables|SetNewVelocityLength_Z (+ _returnvalue_6 (* (Class|FortMovementModeExtBlueprintLibrary|GetDeltaTimefromContext (Variables|GetRuntimeContext) (Variables|FortMovementMode_BaseExtLogic|GetbUseDilatedDeltaTime)) (ScalableFloat|GetValueatLevel (Variables|Default|GetRow_DecelerationRate)))))
                      (LocalVariables|SetNewVelocityLength_Z (Math|Float|Min(Float) (LocalVariables|GetNewVelocityLength_Z) 4e+03)))))
                (LocalVariables|SetNewAcceleration (Variables|GetInAcceleration))
                (Variables|Default|SetStopGlidingMultiplier (Math|Float|Max(Float) (Math|Float|Min(Float) (- (Variables|Default|GetStopGlidingMultiplier) (* (LocalVariables|GetStopGlidingMultiplierDelta) (Class|FortMovementModeExtBlueprintLibrary|GetDeltaTimefromContext (Variables|GetRuntimeContext) (Variables|FortMovementMode_BaseExtLogic|GetbUseDilatedDeltaTime)))) (LocalVariables|GetStopGlidingMultiplierMax)) (LocalVariables|GetStopGlidngMultiplierMin)))
                (else
                  (LocalVariables|SetControlRotation_RotatedByInput (Math|Rotator|MakeRotator (.roll (Math|Rotator|MakeRotator (.roll (Variables|/WingSuit/Gameplay/WingSuit_PhysicsFlying/BpMME_RuntimeData_WingSuit_PhysicsFlying_WraithMovement.BpMME_RuntimeData_WingSuit_PhysicsFlying_WraithMovement_C|GetControlRotation _returnvalue_11)) (.pitch (Variables|/WingSuit/Gameplay/WingSuit_PhysicsFlying/BpMME_RuntimeData_WingSuit_PhysicsFlying_WraithMovement.BpMME_RuntimeData_WingSuit_PhysicsFlying_WraithMovement_C|GetControlRotation _returnvalue_11)) (+ (Math|Float|MapRangeClamped (.Y (Math|Vector2D|BreakVector2D (Variables|/WingSuit/Gameplay/WingSuit_PhysicsFlying/BpMME_RuntimeData_WingSuit_PhysicsFlying_WraithMovement.BpMME_RuntimeData_WingSuit_PhysicsFlying_WraithMovement_C|GetInputVector _returnvalue_11))) 0.5 -0.5 5e+01 -5e+01) (.yaw (Variables|/WingSuit/Gameplay/WingSuit_PhysicsFlying/BpMME_RuntimeData_WingSuit_PhysicsFlying_WraithMovement.BpMME_RuntimeData_WingSuit_PhysicsFlying_WraithMovement_C|GetControlRotation _returnvalue_11))))) (+ (.pitch (Math|Rotator|MakeRotator (.roll (Variables|/WingSuit/Gameplay/WingSuit_PhysicsFlying/BpMME_RuntimeData_WingSuit_PhysicsFlying_WraithMovement.BpMME_RuntimeData_WingSuit_PhysicsFlying_WraithMovement_C|GetControlRotation _returnvalue_11)) (.pitch (Variables|/WingSuit/Gameplay/WingSuit_PhysicsFlying/BpMME_RuntimeData_WingSuit_PhysicsFlying_WraithMovement.BpMME_RuntimeData_WingSuit_PhysicsFlying_WraithMovement_C|GetControlRotation _returnvalue_11)) (+ (Math|Float|MapRangeClamped (.Y (Math|Vector2D|BreakVector2D (Variables|/WingSuit/Gameplay/WingSuit_PhysicsFlying/BpMME_RuntimeData_WingSuit_PhysicsFlying_WraithMovement.BpMME_RuntimeData_WingSuit_PhysicsFlying_WraithMovement_C|GetInputVector _returnvalue_11))) 0.5 -0.5 5e+01 -5e+01) (.yaw (Variables|/WingSuit/Gameplay/WingSuit_PhysicsFlying/BpMME_RuntimeData_WingSuit_PhysicsFlying_WraithMovement.BpMME_RuntimeData_WingSuit_PhysicsFlying_WraithMovement_C|GetControlRotation _returnvalue_11))))) (Variables|Default|GetRow_ArbitraryExtraDownwardPitch)) (.yaw (Math|Rotator|MakeRotator (.roll (Variables|/WingSuit/Gameplay/WingSuit_PhysicsFlying/BpMME_RuntimeData_WingSuit_PhysicsFlying_WraithMovement.BpMME_RuntimeData_WingSuit_PhysicsFlying_WraithMovement_C|GetControlRotation _returnvalue_11)) (.pitch (Variables|/WingSuit/Gameplay/WingSuit_PhysicsFlying/BpMME_RuntimeData_WingSuit_PhysicsFlying_WraithMovement.BpMME_RuntimeData_WingSuit_PhysicsFlying_WraithMovement_C|GetControlRotation _returnvalue_11)) (+ (Math|Float|MapRangeClamped (.Y (Math|Vector2D|BreakVector2D (Variables|/WingSuit/Gameplay/WingSuit_PhysicsFlying/BpMME_RuntimeData_WingSuit_PhysicsFlying_WraithMovement.BpMME_RuntimeData_WingSuit_PhysicsFlying_WraithMovement_C|GetInputVector _returnvalue_11))) 0.5 -0.5 5e+01 -5e+01) (.yaw (Variables|/WingSuit/Gameplay/WingSuit_PhysicsFlying/BpMME_RuntimeData_WingSuit_PhysicsFlying_WraithMovement.BpMME_RuntimeData_WingSuit_PhysicsFlying_WraithMovement_C|GetControlRotation _returnvalue_11)))))) :!label _label_3)
                  (bind _newparam (CallFunction|IsPlayerAttemptingAscending _pitch_2 _pitch))
                  (if (not _newparam)
                    (goto _label))))
              (goto _label_2))
            (goto _label_3))
          (:then_1
            ;[1 1 1] Force facing to determine velocity instead of camera
            (LocalVariables|SetInVelocityVar_Holder (* (Transformation|GetActorForwardVector (Variables|FortMovementMode_BaseExtLogic|GetOwnerPawn)) (Math|Vector|VectorLength (Variables|GetInVelocity)))))
          (:then_2
            ;;[0.08994 0.04175 0.09896] Snap Velocity To Desired Direction
            (LocalVariables|SetThisFrameScaledUpDownInput_Z (* (Math|Conversions|ToFloat(Integer) (+ (Math|Conversions|ToInteger(Boolean) (Variables|/WingSuit/Gameplay/WingSuit_PhysicsFlying/BpMME_RuntimeData_WingSuit_PhysicsFlying_WraithMovement.BpMME_RuntimeData_WingSuit_PhysicsFlying_WraithMovement_C|GetUpInput _returnvalue_11)) (* (Math|Conversions|ToInteger(Boolean) (Variables|/WingSuit/Gameplay/WingSuit_PhysicsFlying/BpMME_RuntimeData_WingSuit_PhysicsFlying_WraithMovement.BpMME_RuntimeData_WingSuit_PhysicsFlying_WraithMovement_C|GetDownInput _returnvalue_11)) -1))) (Math|Float|Clamp(Float) (Math|Float|SelectFloat (ScalableFloat|GetValueatLevel (Variables|UpDownControl|GetUpDownSensitivityMultiplier)) (Math|Curves|GetFloatValue (Variables|UpDownControl|GetUpDownSensitivityCurve) (Math|Float|Clamp(Float) (Math|Float|SafeDivide (Variables|/WingSuit/Gameplay/WingSuit_PhysicsFlying/BpMME_RuntimeData_WingSuit_PhysicsFlying_WraithMovement.BpMME_RuntimeData_WingSuit_PhysicsFlying_WraithMovement_C|GetTotalTimeUpdated _returnvalue_11) (ScalableFloat|GetValueatLevel (Variables|UpDownControl|GetUpDownDampenFallOffDurationSecs))))) (Variables|UpDownControl|GetUseConstantUpDownSensitivity)))))
            (bind _output_get_2 (LocalVariables|SetInVelDirNormalized (Math|Vector|Normalize (LocalVariables|GetInVelocityVar_Holder))))
            (if (Pawn|IsBotControlled (Variables|FortMovementMode_BaseExtLogic|GetOwnerPawn))
              (LocalVariables|SetDesiredTargetVelDirThisFrame (Math|Vector|Normalize (Variables|/WingSuit/Gameplay/WingSuit_PhysicsFlying/BpMME_RuntimeData_WingSuit_PhysicsFlying_WraithMovement.BpMME_RuntimeData_WingSuit_PhysicsFlying_WraithMovement_C|GetRequestedVelocity _returnvalue_11)))
              (else
                (bind _output_get_3 (LocalVariables|SetDesiredTargetVelDirThisFrame (Math|Vector|SelectVector (Math|Vector|SelectVector _output_get_2 _output_get_2 (select (Variables|UpDownControl|GetSeparateUpDownInputVector) (LocalVariables|GetInput2DPressedThisFrame) (LocalVariables|GetInputPressedThisFrame))) (LocalVariables|GetInVelDirNormalized) (> (Math|Vector|VectorLengthSquared (Math|Vector|SelectVector _output_get_2 _output_get_2 (select (Variables|UpDownControl|GetSeparateUpDownInputVector) (LocalVariables|GetInput2DPressedThisFrame) (LocalVariables|GetInputPressedThisFrame)))) 0.001))))))
            (if (Variables|UpDownControl|GetSeparateUpDownInputVector)
              (LocalVariables|SetDesiredTargetVelDirThisFrame (Math|Vector|ClampVectorSize (+ _output_get_3 (Math|Vector|MakeVector 0.0 0.0 (LocalVariables|GetThisFrameScaledUpDownInput_Z))) 0.0 1.0)) ; add back in up/down input after
            )
            (bind _result (CallFunction|LimitUpwardAnglebySpeed (LocalVariables|GetDesiredTargetVelDirThisFrame) (Variables|Default|GetPreviousVelocityLength) (Transformation|GetActorForwardVector (Variables|FortMovementMode_BaseExtLogic|GetOwnerPawn))))
            (LocalVariables|SetDesiredTargetVelDirThisFrame _result) ; add back in up/down input after
            (LocalVariables|SetVelocityDragAlpha (Math|Float|Clamp(Float) (Math|Float|SafeDivide (* (ScalableFloat|GetValueatLevel (Variables|Default|GetVelocityDragAngleRateDegrees)) (Class|FortMovementModeExtBlueprintLibrary|GetDeltaTimefromContext (Variables|GetRuntimeContext) (Variables|FortMovementMode_BaseExtLogic|GetbUseDilatedDeltaTime))) (Math|Trig|Acos(Degrees) (Math|Vector|DotProduct (LocalVariables|GetDesiredTargetVelDirThisFrame) (LocalVariables|GetInVelDirNormalized))))))
            (LocalVariables|SetNewVelocity (Math|Vector|MakeVector (* (Variables|Default|GetPreviousVelocityLength) (.x (Math|Vector|SelectVector (Math|Vector|SlerpVectorToDirection (Math|Vector|SelectVector (LocalVariables|GetInVelDirNormalized) (LocalVariables|GetDesiredTargetVelDirThisFrame) (> (Math|Vector|VectorLengthSquared (LocalVariables|GetInVelDirNormalized)) 0)) (LocalVariables|GetDesiredTargetVelDirThisFrame) (LocalVariables|GetVelocityDragAlpha)) (LocalVariables|GetDesiredTargetVelDirThisFrame) (ScalableFloat|GetValueasBool (Variables|Default|GetEnableVelocityDrag))))) (* (Variables|Default|GetPreviousVelocityLength) (.y (Math|Vector|SelectVector (Math|Vector|SlerpVectorToDirection (Math|Vector|SelectVector (LocalVariables|GetInVelDirNormalized) (LocalVariables|GetDesiredTargetVelDirThisFrame) (> (Math|Vector|VectorLengthSquared (LocalVariables|GetInVelDirNormalized)) 0)) (LocalVariables|GetDesiredTargetVelDirThisFrame) (LocalVariables|GetVelocityDragAlpha)) (LocalVariables|GetDesiredTargetVelDirThisFrame) (ScalableFloat|GetValueasBool (Variables|Default|GetEnableVelocityDrag))))) (* (.z (Math|Vector|SelectVector (Math|Vector|SlerpVectorToDirection (Math|Vector|SelectVector (LocalVariables|GetInVelDirNormalized) (LocalVariables|GetDesiredTargetVelDirThisFrame) (> (Math|Vector|VectorLengthSquared (LocalVariables|GetInVelDirNormalized)) 0)) (LocalVariables|GetDesiredTargetVelDirThisFrame) (LocalVariables|GetVelocityDragAlpha)) (LocalVariables|GetDesiredTargetVelDirThisFrame) (ScalableFloat|GetValueasBool (Variables|Default|GetEnableVelocityDrag)))) (LocalVariables|GetNewVelocityLength_Z))))
            (Variables|SetNewVelocity_X (Math|Vector|MakeVector _returnvalue_x _returnvalue_y (.z (LocalVariables|GetNewVelocity))))
            (bind _asb_weap_wing_suit_physics_flying (Utilities|Casting|CastToB_Weap_WingSuit_PhysicsFlying (Weapon|GetCurrentWeapon (Variables|FortMovementMode_BaseExtLogic|GetOwnerPawn))))
            (Variables|/WingSuit/Gameplay/WingSuit_PhysicsFlying/B_Weap_WingSuit_PhysicsFlying.B_Weap_WingSuit_PhysicsFlying_C|SetVelocity (LocalVariables|GetNewVelocity) _asb_weap_wing_suit_physics_flying)
            (Variables|/WingSuit/Gameplay/WingSuit_PhysicsFlying/B_Weap_WingSuit_PhysicsFlying.B_Weap_WingSuit_PhysicsFlying_C|SetPreviousVelocity (LocalVariables|GetInVelocityVar_Holder) _asb_weap_wing_suit_physics_flying)
            (Variables|/WingSuit/Gameplay/WingSuit_PhysicsFlying/B_Weap_WingSuit_PhysicsFlying.B_Weap_WingSuit_PhysicsFlying_C|SetAcceleration (Variables|GetInAcceleration) _asb_weap_wing_suit_physics_flying)
            (Variables|/WingSuit/Gameplay/WingSuit_PhysicsFlying/B_Weap_WingSuit_PhysicsFlying.B_Weap_WingSuit_PhysicsFlying_C|SetInputVector (Variables|/WingSuit/Gameplay/WingSuit_PhysicsFlying/BpMME_RuntimeData_WingSuit_PhysicsFlying_WraithMovement.BpMME_RuntimeData_WingSuit_PhysicsFlying_WraithMovement_C|GetInputVector _returnvalue_11) _asb_weap_wing_suit_physics_flying)
            (Variables|/WingSuit/Gameplay/WingSuit_PhysicsFlying/B_Weap_WingSuit_PhysicsFlying.B_Weap_WingSuit_PhysicsFlying_C|SetIsUsingGamepad (Variables|/WingSuit/Gameplay/WingSuit_PhysicsFlying/BpMME_RuntimeData_WingSuit_PhysicsFlying_WraithMovement.BpMME_RuntimeData_WingSuit_PhysicsFlying_WraithMovement_C|GetIsOnGamepad _returnvalue_11) _asb_weap_wing_suit_physics_flying)
            (Variables|/WingSuit/Gameplay/WingSuit_PhysicsFlying/B_Weap_WingSuit_PhysicsFlying.B_Weap_WingSuit_PhysicsFlying_C|SetLeftAnalogStickInput (ScalableFloat|GetValueasBool (Variables|Default|GetRow_GamepadLeftAnalogStickFlight)) _asb_weap_wing_suit_physics_flying)
            (Utilities|FlowControl|Sequence
              (:then_0
                ;;[1 1 1] Manage speed history
                (Utilities|Array|Add (Variables|Default|GetHorizontalSpeedHistory) _returnvalue_2)
                (if (> (Utilities|Array|Length (Variables|Default|GetHorizontalSpeedHistory)) 600)
                  (Utilities|Array|RemoveIndex (Variables|Default|GetHorizontalSpeedHistory)))
                (CallFunction|PrintRunningAverageHorizontalSpeed))
              (:then_1
                ;;;[1 1 1] If going slow, more gravity
                (Development|PrintString 0 (Utilities|String|Append "Current Horizontal Speed: " (Utilities|String|ToString(Float) _returnvalue_2)) true true "(R=0.000000,G=0.660000,B=1.000000,A=1.000000)" 3e+01 "WingSuit_HorizontalVelocity")
                (Development|PrintString 0 (Utilities|String|Append "Total Speed: " (Utilities|String|ToString(Float) (Variables|Default|GetPreviousVelocityLength))) true true "(R=0.000000,G=0.660000,B=1.000000,A=1.000000)" 3e+01 "WingSuit_TotalVelocity")
                (if (> (LocalVariables|GetNewVelocityLength_Z) 1.25e+03)
                  _)
                (LocalVariables|SetNewVelocity (+ (LocalVariables|GetNewVelocity) (* (Class|FortMovementModeExtBlueprintLibrary|GetDeltaTimefromContext (Variables|GetRuntimeContext) (Variables|FortMovementMode_BaseExtLogic|GetbUseDilatedDeltaTime)) "0.000000,0.000000,-4500.000000")))
                (if false
                  (if (Utilities|Enum|Equal(Enum) (Networking|GetLocalRole (Variables|FortMovementMode_BaseExtLogic|GetOwnerPawn)) "ROLE_AutonomousProxy")
                    (Rendering|Debug|DrawDebugArrow (Variables|FortMovementMode_BaseExtLogic|GetOwnerPawn) (CallFunction|GetDebugVisualBaseLocation "0.000000,0.000000,150.000000") (+ (CallFunction|GetDebugVisualBaseLocation "0.000000,0.000000,150.000000") (* (LocalVariables|GetDesiredTargetVelDirThisFrame) 1e+02)) 1.0 "(R=1.000000,G=0.043382,B=0.000000,A=1.000000)")
                    (Rendering|Debug|DrawDebugArrow (Variables|FortMovementMode_BaseExtLogic|GetOwnerPawn) (CallFunction|GetDebugVisualBaseLocation "0.000000,0.000000,200.000000") (+ (CallFunction|GetDebugVisualBaseLocation "0.000000,0.000000,200.000000") (* (Math|Vector|Normalize (LocalVariables|GetNewVelocity)) 1e+02)) 0.0 "(R=0.155667,G=0.000000,B=1.000000,A=1.000000)")
                    (else
                      (if (Utilities|Enum|Equal(Enum) (Networking|GetLocalRole (Variables|FortMovementMode_BaseExtLogic|GetOwnerPawn)) "ROLE_AutonomousProxy") :!label _label_4
                        _)))
                  (return (LocalVariables|GetNewVelocity) (LocalVariables|GetNewAcceleration)))
                (goto _label_4))))
          (:then_3
            (bind _output_get_4 (Variables|Default|SetServerTotalTime (+ (Variables|Default|GetServerTotalTime) (Class|FortMovementModeExtBlueprintLibrary|GetDeltaTimefromContext (Variables|GetRuntimeContext) (Variables|FortMovementMode_BaseExtLogic|GetbUseDilatedDeltaTime))))))))
      (goto _label_5))))

(fn UpdateBeforeCharacterMovement
  (params (RuntimeContext :type MMERuntime%20Context%20Structure%20%28by%20ref%29))
  (bind _ownerpawn (Variables|FortMovementMode_BaseExtLogic|GetOwnerPawn))
  (bind _returnvalue (Weapon|GetCurrentWeapon _ownerpawn))
  (Utilities|Casting|CastToB_Weap_WingSuit_PhysicsFlying _returnvalue
    (:then
      (FortWeapon|UseWeaponDurability(Explicit) _returnvalue 0.25)
      ;[1 1 1] Why does a world trace channel not work here
      (bind (_outhit _returnvalue_2) (Collision|LineTraceByChannel (Variables|FortMovementMode_BaseExtLogic|GetOwnerPawn) (Transformation|GetActorLocation (Variables|FortMovementMode_BaseExtLogic|GetOwnerPawn)) (Utilities|Operators|Add (Transformation|GetActorLocation (Variables|FortMovementMode_BaseExtLogic|GetOwnerPawn)) (Math|Vector|MakeVector 0.0 0.0 (* (+ (Components|Capsule|GetScaledCapsuleHalfHeight (Variables|/Script/Engine.Character|GetCapsuleComponent (Variables|FortMovementMode_BaseExtLogic|GetOwnerPawn))) 2e+01) -1.0))) "TraceTypeQuery1" true))
      (if _returnvalue_2
        (Variables|Default|SetRollLanding true)
        (else
          (return "Active"))))
    (:CastFailed
      _))
  (return "Ended"))

(fn DoHoverLogic
  (params (MMERuntimeContext :type MMERuntime%20Context%20Structure%20%28by%20ref%29))
  (if (ScalableFloat|GetValueasBool (Variables|Hover|GetEnableHoverLogic))
    (bind (_outhit _returnvalue) (Collision|LineTraceByChannel :WorldContextObject (Variables|FortMovementMode_BaseExtLogic|GetOwnerPawn) :Start (- (Transformation|GetActorLocation (Variables|FortMovementMode_BaseExtLogic|GetOwnerPawn)) (Math|Vector|MakeVector 0.0 0.0 (Components|Capsule|GetScaledCapsuleHalfHeight (Variables|/Script/Engine.Character|GetCapsuleComponent (Variables|FortMovementMode_BaseExtLogic|GetOwnerPawn))))) :End (- (- (Transformation|GetActorLocation (Variables|FortMovementMode_BaseExtLogic|GetOwnerPawn)) (Math|Vector|MakeVector 0.0 0.0 (Components|Capsule|GetScaledCapsuleHalfHeight (Variables|/Script/Engine.Character|GetCapsuleComponent (Variables|FortMovementMode_BaseExtLogic|GetOwnerPawn))))) (Math|Vector|MakeVector 0.0 0.0 (ScalableFloat|GetValueatLevel (Variables|Hover|GetGroundHoverDistance)))) :TraceChannel "TraceTypeQuery1" :bTraceComplex false :DrawDebugType (select (Variables|Hover|GetDebugHoverLogic) "ForOneFrame" "None") :bIgnoreSelf true :TraceColor "(R=1.000000,G=0.000000,B=0.000000,A=1.000000)" :TraceHitColor "(R=0.000000,G=1.000000,B=0.000000,A=1.000000)" :DrawTime 5.0))
    (if _returnvalue
      (if (Collision|BreakHitResult _outhit)
        (Transformation|SetActorLocation (Variables|FortMovementMode_BaseExtLogic|GetOwnerPawn) (+ (Transformation|GetActorLocation (Variables|FortMovementMode_BaseExtLogic|GetOwnerPawn)) (Math|Vector|MakeVector 0.0 0.0 (Math|Float|Min(Float) (* (Class|FortMovementModeExtBlueprintLibrary|GetDeltaTimefromContext MMERuntimeContext (Variables|FortMovementMode_BaseExtLogic|GetbUseDilatedDeltaTime)) (ScalableFloat|GetValueatLevel (Variables|Hover|GetHoverSpeed))) (- (ScalableFloat|GetValueatLevel (Variables|Hover|GetGroundHoverDistance)) (.Distance (Collision|BreakHitResult _outhit)))))) true false) ; sweep so we collide with world if there is an issue.
      ))))

(fn GetDebugVisualBaseLocation
  (params (Offset :type Vector))
  (return (+ Offset (Transformation|GetActorLocation (Variables|FortMovementMode_BaseExtLogic|GetOwnerPawn)))))

(fn GetLerpedSpeedForThisFrame
  (params (InNewVelocityLength :type Float%20%28double-precision%29) (InRuntimeContext :type MMERuntime%20Context%20Structure))
  (bind _boostfalloffduration (Variables|Default|GetBoostFallOffDuration))
  (bind _returnvalue (ScalableFloat|GetValueatLevel _boostfalloffduration))
  (bind _fastflightspeed_runtime (Variables|Default|GetFastFlightSpeed_Runtime))
  (bind _boostspeedfalloffcurve (Variables|Default|GetBoostSpeedFallOffCurve))
  (bind _slowflightspeed_runtime (Variables|Default|GetSlowFlightSpeed_Runtime))
  (if (GameplayTags|HasMatchingGameplayTag (Variables|FortMovementMode_BaseExtLogic|GetOwnerPawn) "(TagName=\"MistRaven.AllAscendedEquipped\")")
    (Variables|Default|SetFastFlightSpeed_Runtime (ScalableFloat|GetValueasDouble (Variables|Default|GetAscended_FastWraithFlightSpeed)))
    (Variables|Default|SetSlowFlightSpeed_Runtime (ScalableFloat|GetValueasDouble (Variables|Default|GetAscended_SlowWraithFlightSpeed)))
    (else
      (Variables|Default|SetFastFlightSpeed_Runtime (ScalableFloat|GetValueasDouble (Variables|Default|GetFastWraithFlightSpeed)))
      (Variables|Default|SetSlowFlightSpeed_Runtime (ScalableFloat|GetValueasDouble (Variables|Default|GetSlowWraithFlightSpeed)))))
  (bind _returnvalue_2 (Class|FortMovementModeExtBlueprintLibrary|GetRuntimeDatafromContext (Variables|GetInRuntimeContext) "/WingSuit/Gameplay/WingSuit_PhysicsFlying/BpMME_RuntimeData_WingSuit_PhysicsFlying_WraithMovement.BpMME_RuntimeData_WingSuit_PhysicsFlying_WraithMovement_C"))
  (if false
    (Development|PrintString (Variables|FortMovementMode_BaseExtLogic|GetOwnerPawn) (Utilities|String|Append "wraith speed " (Utilities|String|ToString(Float) (Math|Float|Lerp _slowflightspeed_runtime _fastflightspeed_runtime (Math|Curves|GetFloatValue _boostspeedfalloffcurve (Math|Float|Clamp(Float) (Math|Float|SafeDivide (Variables|/WingSuit/Gameplay/WingSuit_PhysicsFlying/BpMME_RuntimeData_WingSuit_PhysicsFlying_WraithMovement.BpMME_RuntimeData_WingSuit_PhysicsFlying_WraithMovement_C|GetTotalTimeUpdated _returnvalue_2) _returnvalue))))))))
  (return (Math|Float|Clamp(Float) (Variables|GetInNewVelocityLength) 0.0 (Math|Float|Lerp _slowflightspeed_runtime _fastflightspeed_runtime (Math|Curves|GetFloatValue _boostspeedfalloffcurve (Math|Float|Clamp(Float) (Math|Float|SafeDivide (Variables|/WingSuit/Gameplay/WingSuit_PhysicsFlying/BpMME_RuntimeData_WingSuit_PhysicsFlying_WraithMovement.BpMME_RuntimeData_WingSuit_PhysicsFlying_WraithMovement_C|GetTotalTimeUpdated _returnvalue_2) _returnvalue)))))))

(fn GetZClampSpeedHeight
  (if (ScalableFloat|GetValueasBool (Variables|ZClamp|GetZSpeedClampEnabled))
    (return (ScalableFloat|GetValueatLevel (select (GameplayTags|HasMatchingGameplayTag (Variables|FortMovementMode_BaseExtLogic|GetOwnerPawn) "(TagName=\"MistRaven.AllAscendedEquipped\")") (Variables|ZClamp|GetZSpeedClamp_Combined) (Variables|ZClamp|GetZSpeedClamp_Normal))))
    (else
      (return -1.0))))

(fn HandleHit_NormalBump
  (params (InVelocity :type Vector) (OldPosition :type Vector) (NewLocation :type Vector) (AdjustmentVector :type Vector) (Hit :type Hit%20Result%20Structure) (DebugIndex :type Integer))
  (bind _normalbumpdistance (Variables|Default|GetNormalBumpDistance))
  (bind _returnvalue (ScalableFloat|GetValueatLevel _normalbumpdistance))
  (bind _ownerpawn (Variables|FortMovementMode_BaseExtLogic|GetOwnerPawn))
  (bind _returnvalue_2 (Transformation|GetActorLocation _ownerpawn))
  (bind (_bblockinghit _binitialoverlap _time _distance _location _impactpoint _normal) (Collision|BreakHitResult Hit))
  (bind _adjustmentvector (Variables|GetAdjustmentVector))
  (bind _returnvalue_3 (Math|Vector|ProjectVectorontoPlane _adjustmentvector _normal))
  (bind _returnvalue_4 (- 1.0 _time))
  (bind _returnvalue_5 (* _returnvalue_3 _returnvalue_4))
  (bind _returnvalue_6 (* _normal _returnvalue))
  (bind _returnvalue_7 (+ _returnvalue_2 _returnvalue_6))
  (if (< _time 1.0)
    ;[1 1 1] Normal Bump
    (if (ScalableFloat|GetValueasBool (Variables|Default|GetDebugHitLogicEnabled))
      (Rendering|Debug|DrawDebugArrow (Variables|FortMovementMode_BaseExtLogic|GetOwnerPawn) _returnvalue_2 _returnvalue_7 0.0 "(R=1.000000,G=0.928018,B=0.000000,A=1.000000)" 15.0))
    (Transformation|SetActorLocation _ownerpawn _returnvalue_7 true false) ; bump normal distance
    ;[1 1 1] Slide!
    (bind _sweephitresult (Transformation|SetActorLocation (Variables|FortMovementMode_BaseExtLogic|GetOwnerPawn) (+ (Transformation|GetActorLocation (Variables|FortMovementMode_BaseExtLogic|GetOwnerPawn)) _returnvalue_5) true false)) ; sweep in case we hit something else.
    (if (ScalableFloat|GetValueasBool (Variables|Default|GetDebugHitLogicEnabled))
      (Rendering|Debug|DrawDebugString (Variables|FortMovementMode_BaseExtLogic|GetOwnerPawn) (Math|Vector|MakeVector 0.0 0.0 (+ 1e+02 (* DebugIndex 50))) (Utilities|String|Append "hit slide " (Utilities|String|ToString(Integer) DebugIndex)) (Variables|FortMovementMode_BaseExtLogic|GetOwnerPawn)))
    (return true (Transformation|GetVelocity (Variables|FortMovementMode_BaseExtLogic|GetOwnerPawn)) (Variables|GetNewLocation) (Transformation|GetActorLocation (Variables|FortMovementMode_BaseExtLogic|GetOwnerPawn)) _returnvalue_5 _sweephitresult)
    (else
      (return false (Variables|GetInVelocity) (Variables|GetOldPosition) (Variables|GetNewLocation) (Variables|GetAdjustmentVector) (Variables|GetHit)))))

(fn HandleHit_RicochetVelocity
  (params (InVelocity :type Vector) (OldPosition :type Vector) (NewLocation :type Vector) (AdjustmentVector :type Vector) (Hit :type Hit%20Result%20Structure) (DebugIndex :type Integer))
  (bind _invelocity (Variables|GetInVelocity))
  (bind _returnvalue (Math|Vector|GetReflectionVector _invelocity (.Normal (Collision|BreakHitResult (Variables|GetHit)))))
  (bind _ownerpawn (Variables|FortMovementMode_BaseExtLogic|GetOwnerPawn))
  (bind _returnvalue_2 (Transformation|GetActorLocation _ownerpawn))
  (bind _hit (Variables|GetHit))
  (bind (_bblockinghit _binitialoverlap _time _distance _location _impactpoint _normal) (Collision|BreakHitResult _hit))
  (bind _adjustmentvector (Variables|GetAdjustmentVector))
  (bind _returnvalue_3 (Math|Vector|ProjectVectorontoPlane _adjustmentvector _normal))
  (bind _returnvalue_4 (- 1.0 _time))
  (bind _returnvalue_5 (* _returnvalue_3 _returnvalue_4))
  (if (< _time 1.0)
    (if (ScalableFloat|GetValueasBool (Variables|Default|GetDebugHitLogicEnabled))
      (Rendering|Debug|DrawDebugArrow (Variables|FortMovementMode_BaseExtLogic|GetOwnerPawn) _returnvalue_2 (+ _returnvalue_2 (* _normal 5e+01)) 0.0 "(R=1.000000,G=0.928018,B=0.000000,A=1.000000)" 15.0)
      (Rendering|Debug|DrawDebugArrow (Variables|FortMovementMode_BaseExtLogic|GetOwnerPawn) _returnvalue_2 (+ (Transformation|GetActorLocation (Variables|FortMovementMode_BaseExtLogic|GetOwnerPawn)) (* (Math|Vector|Normalize _returnvalue) 5e+01)) 0.0 "(R=1.000000,G=0.000000,B=0.000000,A=1.000000)" 15.0)
      (Rendering|Debug|DrawDebugArrow (Variables|FortMovementMode_BaseExtLogic|GetOwnerPawn) (+ _returnvalue_2 (* (Math|Vector|Normalize (Variables|GetInVelocity)) -5e+01)) _returnvalue_2 0.0 "(R=0.000000,G=0.167813,B=1.000000,A=1.000000)" 15.0))
    (Variables|/Script/Engine.MovementComponent|SetVelocity (+ (- (* _returnvalue (ScalableFloat|GetValueatLevel (Variables|Default|GetRicochetSpeedDampening))) (Math|Vector|ProjectVectorontoPlane _returnvalue _normal)) (* (Math|Vector|ProjectVectorontoPlane _returnvalue _normal) (ScalableFloat|GetValueatLevel (Variables|Default|GetRicochetSpeedNormalAlpha)))) (Variables|/Script/Engine.Character|GetCharacterMovement (Variables|FortMovementMode_BaseExtLogic|GetOwnerPawn)))
    ;[1 1 1] Slide!
    (bind _sweephitresult (Transformation|SetActorLocation (Variables|FortMovementMode_BaseExtLogic|GetOwnerPawn) (+ (Transformation|GetActorLocation (Variables|FortMovementMode_BaseExtLogic|GetOwnerPawn)) _returnvalue_5) true false)) ; sweep in case we hit something else.
    (if (ScalableFloat|GetValueasBool (Variables|Default|GetDebugHitLogicEnabled))
      (Rendering|Debug|DrawDebugString (Variables|FortMovementMode_BaseExtLogic|GetOwnerPawn) (Math|Vector|MakeVector 0.0 0.0 (+ 1e+02 (* DebugIndex 50))) (Utilities|String|Append "hit slide " (Utilities|String|ToString(Integer) DebugIndex)) (Variables|FortMovementMode_BaseExtLogic|GetOwnerPawn)))
    (return true (Transformation|GetVelocity (Variables|FortMovementMode_BaseExtLogic|GetOwnerPawn)) (Variables|GetNewLocation) (Transformation|GetActorLocation (Variables|FortMovementMode_BaseExtLogic|GetOwnerPawn)) _returnvalue_5 _sweephitresult)
    (else
      (return false (Variables|GetInVelocity) (Variables|GetOldPosition) (Variables|GetNewLocation) (Variables|GetAdjustmentVector) (Variables|GetHit)))))

(fn HandleHit_Slide
  (params (InVelocity :type Vector) (OldPosition :type Vector) (NewLocation :type Vector) (AdjustmentVector :type Vector) (Hit :type Hit%20Result%20Structure) (DebugIndex :type Integer))
  (bind (_bblockinghit _binitialoverlap _time _distance _location _impactpoint _normal) (Collision|BreakHitResult Hit))
  (bind _returnvalue (Math|Vector|ProjectVectorontoPlane AdjustmentVector _normal))
  (bind _returnvalue_2 (- 1.0 _time))
  (bind _returnvalue_3 (* _returnvalue _returnvalue_2))
  (bind _returnvalue_4 (< _time 1.0))
  ;[1 1 1] Fake SlideVectorLogic since actual functions do not seem exposed.
  (if _returnvalue_4
    (bind _sweephitresult (Transformation|SetActorLocation (Variables|FortMovementMode_BaseExtLogic|GetOwnerPawn) (+ NewLocation _returnvalue_3) true false)) ; sweep in case we hit something else.
    (if (ScalableFloat|GetValueasBool (Variables|Default|GetDebugHitLogicEnabled))
      (Rendering|Debug|DrawDebugString (Variables|FortMovementMode_BaseExtLogic|GetOwnerPawn) (Math|Vector|MakeVector 0.0 0.0 (+ 1e+02 (* DebugIndex 50))) (Utilities|String|Append "hit slide " (Utilities|String|ToString(Integer) DebugIndex)) (Variables|FortMovementMode_BaseExtLogic|GetOwnerPawn)))
    (return _returnvalue_4 (Transformation|GetVelocity (Variables|FortMovementMode_BaseExtLogic|GetOwnerPawn)) NewLocation (Transformation|GetActorLocation (Variables|FortMovementMode_BaseExtLogic|GetOwnerPawn)) _returnvalue_3 _sweephitresult)
    (else
      (return false (Variables|GetInVelocity) (Variables|GetOldPosition) (Variables|GetNewLocation) (Variables|GetAdjustmentVector) (Variables|GetHit)))))

(fn IsPlayerAttemptingAscending
  (params (PlayerPitch :type Float%20%28double-precision%29) (CameraPitch :type Float%20%28double-precision%29))
  (LocalVariables|SetNewCameraPitch CameraPitch)
  (if (> CameraPitch 9e+01)
    (LocalVariables|SetNewCameraPitch (- CameraPitch 3.6e+02)))
  (return (> (LocalVariables|GetNewCameraPitch) PlayerPitch)))

(fn LimitUpwardAngleBySpeed
  (params (DesiredVel :type Vector) (CurrentSpeed :type Float%20%28double-precision%29) (CharacterForward :type Vector))
  (if (not (Math|Vector|VectorIsNearlyZero DesiredVel))
    (bind _returnvalue (Fortnite|GetAngleDegrees DesiredVel (0.0 0.0 1.0)))
    (bind _output_get (LocalVariables|SetMaxUpwardsAngleDegrees (Math|Curves|GetFloatValue (Variables|Default|GetVelocityMapAngle_Degrees_Curve) CurrentSpeed)))
    (if (> (- 9e+01 _returnvalue) _output_get)
      (LocalVariables|SetNewRotationAmount (- (- 9e+01 _returnvalue) _output_get))
      (bind _output_get_2 (LocalVariables|SetRotationAxis (Math|Vector|CrossProduct "0.000000,0.000000,1.000000" (Variables|GetDesiredVel))))
      (if (Math|Vector|VectorIsNearlyZero _output_get_2)
        (LocalVariables|SetRotationAxis (Math|Vector|CrossProduct "0.000000,0.000000,1.000000" (Variables|GetCharacterForward))))
      (return (Math|Vector|RotateVectorAroundAxis (Variables|GetDesiredVel) (LocalVariables|GetNewRotationAmount) (LocalVariables|GetRotationAxis)))
      (else
        _)))
  (return (Variables|GetDesiredVel)))

(fn PrintRunningAverageHorizontalSpeed
  (if (Utilities|Array|IsNotEmpty (Variables|Default|GetHorizontalSpeedHistory))
    (for _array_element (Variables|Default|GetHorizontalSpeedHistory)
      (LocalVariables|SetRunningTotal (+ _array_element (LocalVariables|GetRunningTotal))))
    (Development|PrintString 0 (Utilities|String|Append "Horizontal Average Speed: " (Utilities|String|ToString(Float) (/ (LocalVariables|GetRunningTotal) (Utilities|Array|Length (Variables|Default|GetHorizontalSpeedHistory))))) true true "(R=0.000000,G=0.660000,B=1.000000,A=1.000000)" 2e+01 "AverageSpeed")))
