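; Movement mode extension for landing from wingsuit physics flying, with a hook for energy gain.
; On activation it calls the parent, finds the Fort energy component on the owner pawn and, if valid, checks the EnergyGainEnabled scalable float.
; No action follows that check, so no energy is granted here.
; Deactivation only calls the parent.
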(event EventOnExtensionActivate
  (params (RuntimeContext :type MMERuntime%20Context%20Structure%20%28by%20ref%29))
  (|CallParentFunction RuntimeContext)
  (Utilities|IsValid (Actor|GetComponentbyClass (Variables|FortMovementMode_BaseExtLogic|GetOwnerPawn) "/Script/EnergyRuntime.FortComponent_Energy")
    (:Is%20Valid
      (if (ScalableFloat|GetValueasBool (Variables|Default|GetEnergyGainEnabled))
        _))))

(event EventOnExtensionDeactivate
  (params (RuntimeContext :type MMERuntime%20Context%20Structure%20%28by%20ref%29))
  (|CallParentFunction RuntimeContext))
