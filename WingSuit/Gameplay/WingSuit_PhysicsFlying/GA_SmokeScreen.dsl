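; Gameplay ability for the wingsuit smoke screen, active only when the avatar pawn has the flight tag.
; On activation it applies the smoke effect, adds a gameplay cue and plays a showoff montage.
; Wait tasks end the ability when the flight tag is removed or the secondary-released event arrives.
; On end it removes effect and cue, ends the release-event task and stops the montage.
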(event Ability|EventActivateAbility
  (bind _returnvalue (GameplayTags|HasMatchingGameplayTag(Message) (CallFunction|GetAvatarasFortPlayerPawn) (Variables|Default|GetFlightTag)))
  (if _returnvalue
    (bind _returnvalue_2 (Ability|ApplyGameplayEffectToOwner "/WingSuit/Gameplay/WingSuit_PhysicsFlying/GE_SmokeScreening.GE_SmokeScreening_C"))
    (Variables|Default|SetgameplayEffectForSmoke _returnvalue_2)
    (Ability|AddGameplayCueToOwner (Variables|Default|GetGameplay%20Cue%20Tag))
    (bind _asynctaskproxy (Ability|Tasks|WaitGameplayTagRemove (Variables|Default|GetFlightTag) 0 true)
      (:then
        (Variables|Default|SetRemoveFlightWatcher _asynctaskproxy)
        (bind _asynctaskproxy_2 (Ability|Tasks|WaitGameplayEvent "(TagName=\"Event.Secondary.Released\")" 0 true)
          (:then
            (Variables|Default|SetEndAbilityEvent _asynctaskproxy_2)
            (bind _asynctaskproxy_3 (Ability|Tasks|PlayMontageAndWaitNotify "Wingsuit_Smoke_Showoff" (Variables|Default|GetShowoffMontageToPlay) 1.0 "None" false))
            (Variables|Default|SetPlayingMontage _asynctaskproxy_3))
          (:EventReceived
            _)))
      (:Removed
        _))
    (Ability|EndAbility)))

(event Ability|EventOnEndAbility
  (params (bWasCancelled :type Boolean))
  (Ability|RemoveGameplayEffectFromOwnerWithHandle (Variables|Default|GetgameplayEffectForSmoke))
  (Ability|RemoveGameplayCueFromOwner (Variables|Default|GetGameplay%20Cue%20Tag))
  (GameplayTasks|EndTask (Variables|Default|GetEndAbilityEvent))
  (Animation|Montage|MontageStop (Components|SkeletalMesh|GetAnimInstance (Variables|/Script/Engine.Character|GetMesh (CallFunction|GetAvatarasFortPlayerPawn))) 0.2 (Variables|Default|GetShowoffMontageToPlay)))
