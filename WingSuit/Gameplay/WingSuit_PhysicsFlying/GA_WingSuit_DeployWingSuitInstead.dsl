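; Gameplay ability that deploys the physics-flying wingsuit on the player pawn.
; On activation without authority, it tries to activate the wingsuit movement mode extension and applies a fall immunity effect.
; It plays the current wingsuit weapon's starting montage, whose notifies add the unfurl tag and send WingSuit.BeginFlight, then remove the tag and end the ability.
; A fallback timer also ends the ability, and ending always clears the timer and tag.
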(event Custom|EndUnfurl
  (Ability|GameplayEffect|RemoveLooseGameplayTags (Variables|Default|GetPlayerPawn) (Variables|Default|GetUnfurlTag) true)
  (Ability|EndAbility))

(event Custom|Unfurl
  (Ability|GameplayEffect|AddLooseGameplayTags (Variables|Default|GetPlayerPawn) (Variables|Default|GetUnfurlTag) true)
  (Ability|SendGameplayEventtoActor (Variables|Default|GetPlayerPawn) "(TagName=\"WingSuit.BeginFlight\")"))

(event Ability|EventActivateAbility
  (bind _returnvalue (Utilities|Time|SetTimerbyEvent (delegate Custom|EventuallyEnd) (ScalableFloat|GetValueatLevel (Variables|Default|GetEventualEndTime))))
  (Variables|Default|SetEventualEndTimer _returnvalue)
  (Variables|Default|SetPlayerPawn (CallFunction|GetAvatarasFortPlayerPawn))
  (Ability|HasAuthority
    (:True
      _)
    (:False
      (Class|FortMovementModeExtLogicWingSuit|TryActivateWingSuitMME (Variables|Default|GetPlayerPawn) "/WingSuit/Gameplay/WingSuit_PhysicsFlying/Native/FMM_ExtLogicWingSuit.FMM_ExtLogicWingSuit_C" (Utilities|Struct|MakeFortMovementModeWingSuitCreationData false))
      (Fortnite|ApplyGameplayEffectsToActor (Variables|Default|GetPlayerPawn) (Variables|Default|GetFallImmunityGE) (Transformation|GetActorLocation (Variables|Default|GetPlayerPawn)) false 0 (Ability|GetAbilitySystemComponent (Variables|Default|GetPlayerPawn)))))
  (bind _asb_weap_wing_suit_physics_flying (Utilities|Casting|CastToB_Weap_WingSuit_PhysicsFlying (Weapon|GetCurrentWeapon (Variables|Default|GetPlayerPawn))))
  (bind _retval (Class|BWeapWingSuitPhysicsFlying|GetStartingAnim _asb_weap_wing_suit_physics_flying))
  (bind (_asynctaskproxy _notifytag) (Ability|Tasks|PlayMontageAndWaitNotify "WingSuitUnfurl" _retval 1.0 "" false)
    (:OnNotifyBegin
      (if (== _notifytag "(TagName=\"WingSuit.AnimNotify.Unfurl\")")
        (CallFunction|Unfurl)
        (elif (== _notifytag "(TagName=\"WingSuit.AnimNotify.EndUnfurl\")")
          (CallFunction|EndUnfurl))))))

(event Custom|EventuallyEnd
  (Ability|EndAbility))

(event Ability|EventOnEndAbility
  (params (bWasCancelled :type Boolean))
  (Utilities|Time|ClearandInvalidateTimerbyHandle (Variables|Default|GetEventualEndTimer))
  (Ability|GameplayEffect|RemoveLooseGameplayTags (Variables|Default|GetPlayerPawn) (Variables|Default|GetUnfurlTag) true))
