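; Gameplay ability for active wingsuit flight, running until a WingSuit.End event ends it.
; Activation applies flight and smoke-screen effects, a cue, tail VFX, input-specific turn scaling and a boost camera that switches on the Unfurl tag.
; Gameplay events and the MovingFast tag toggle a speed lens effect; an optional debug timer prints speeds.
; Ending stops the movement-mode extension, undoes flight state, randomizes smoke color and handles free deploys, cooldowns and delete-on-land item removal.
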(event Custom|GetMME
  (bind _returnvalue (CallFunction|GetAvatarasFortPlayerPawn))
  (bind _returnvalue_2 (Ability|GetFortAbilitySystemComponent _returnvalue))
  (bind _outabilityhandles (GameplayAbilities|GetAllAbilities _returnvalue_2))
  (bind _array_element (Utilities|Array|ForEachLoopwithBreak _outabilityhandles)
    (:LoopBody
      (bind _asga_wing_suit_activate (Utilities|Casting|CastToGA_WingSuit_Activate (.ReturnValue (Ability|GameplayAbility|GetGameplayAbilityfromSpecHandle _returnvalue_2 _array_element))))
      (Variables|Default|SetMME (Variables|/WingSuit/Gameplay/WingSuit_PhysicsFlying/Native/GA_WingSuit_Activate.GA_WingSuit_Activate_C|GetMME _asga_wing_suit_activate)))))

(event Ability|EventActivateAbility
  (bind _returnvalue (CallFunction|GetAvatarasFortPlayerPawn))
  (|CallParentFunction)
  (Ability|Async|WaitGameplayEventToActor _returnvalue "(TagName=\"WingSuit.End\")" true
    (:then
      (CallFunction|GetMME)
      (bind _asb_weap_wing_suit_physics_flying (Utilities|Casting|CastToB_Weap_WingSuit_PhysicsFlying (Weapon|GetCurrentWeapon _returnvalue)))
      (Variables|Default|SetAs%20B%20Weap%20Wing%20Suit%20Physics%20Flying _asb_weap_wing_suit_physics_flying)
      (Ability|Tasks|WaitGameplayEvent "(TagName=\"WingSuit.SlowVisualEffects\")"
        (:then
          (Ability|Tasks|WaitGameplayEvent "(TagName=\"WingSuit.FastVisualEffects\")"
            (:then
              (bind _returnvalue_2 (Ability|ApplyGameplayEffectToOwner "/WingSuit/Gameplay/WingSuit_PhysicsFlying/GE_WingSuit_FlyingActive.GE_WingSuit_FlyingActive_C"))
              (Variables|Default|SetgameplayEffectForFlight _returnvalue_2)
              (bind _returnvalue_3 (Ability|ApplyGameplayEffectToOwner "/WingSuit/Gameplay/WingSuit_PhysicsFlying/GE_SmokeScreening.GE_SmokeScreening_C"))
              (Variables|Default|SetgameplayEffectForSmoke _returnvalue_3)
              (Ability|AddGameplayCueToOwner (Variables|Default|GetGameplay%20Cue%20Tag))
              (bind _asfort_player_controller (Utilities|Casting|CastToFortPlayerController (Pawn|GetController (CallFunction|GetAvatarasFortPlayerPawn))))
              (Variables|Default|SetLocal%20Player%20Controller _asfort_player_controller)
              (Variables|Default|SetIsGamepad (Input|IsUsingGamepad _asfort_player_controller))
              (Camera|PushCameraMode "/WingSuit/Gameplay/WingSuit_PhysicsFlying/Camera/CameraMode_WingSuit_InitialBoostOnlyv3.CameraMode_WingSuit_InitialBoostOnlyv3_C")
              (Ability|Tasks|WaitGameplayTagAdd "(TagName=\"WingSuit.Unfurl\")" 0 true
                (:then
                  (if (Variables|Default|GetIsGamepad)
                    (Input|SetInputPitchScale _asfort_player_controller (ScalableFloat|GetValueatLevel (Variables|Default|GetRow_TurnSpeedPitch_Gamepad)))
                    (Input|SetInputYawScale _asfort_player_controller (ScalableFloat|GetValueatLevel (Variables|Default|GetRow_TurnSpeedYaw_Gamepad)))
                    (else
                      (Input|SetInputPitchScale _asfort_player_controller (ScalableFloat|GetValueatLevel (Variables|Default|GetRow_TurnSpeedPitch_MouseKeyboard)))
                      (Input|SetInputYawScale _asfort_player_controller (ScalableFloat|GetValueatLevel (Variables|Default|GetRow_TurnSpeedYaw_MouseKeyboard)))))
                  (Variables|Default|SetStartTime (Utilities|Time|GetTimeSeconds))
                  (bind _returnvalue_4 (Niagara|SpawnSystemAttached "/WingSuit/VFX/NS_WingSuit_Tail.NS_WingSuit_Tail" (Variables|/Script/Engine.Character|GetMesh (CallFunction|GetAvatarasFortPlayerPawn)) "root"))
                  (Variables|Default|SetCurrentVfx _returnvalue_4)
                  (if (ScalableFloat|GetValueasBool (Variables|Default|GetDebugSpeed))
                    (Utilities|Time|SetTimerbyEvent (delegate Custom|Tick) 0.1 true true))
                  (Ability|Tasks|WaitGameplayTagAdd "(TagName=\"WingSuit.MovingFast\")"
                    (:then
                      (Ability|Tasks|WaitGameplayTagRemove "(TagName=\"WingSuit.MovingFast\")"
                        (:Removed
                          ;[1 1 1] REMOVE THIS WITH JEFF SENDING THE EVENTS
                          (CallFunction|RemoveVisualEffects))))
                    (:Added
                      (CallFunction|AddFastVisualEffects))))
                (:Added
                  (Camera|PopCameraMode)
                  (if (Variables|Default|GetIsGamepad)
                    (Camera|PushCameraMode "/WingSuit/Gameplay/WingSuit_PhysicsFlying/Camera/CameraMode_WingSuit_PhysicsFlying_Gamepad.CameraMode_WingSuit_PhysicsFlying_Gamepad_C")
                    (else
                      (Camera|PushCameraMode "/WingSuit/Gameplay/WingSuit_PhysicsFlying/Camera/CameraMode_WingSuit_PhysicsFlying.CameraMode_WingSuit_PhysicsFlying_C"))))))
            (:EventReceived
              (CallFunction|AddFastVisualEffects))))
        (:EventReceived
          (CallFunction|RemoveVisualEffects))))
    (:EventReceived
      (Ability|EndAbility))))

(event Custom|Tick
  (bind _returnvalue (CallFunction|GetAvatarasFortPlayerPawn))
  (bind _returnvalue_2 (Transformation|GetVelocity _returnvalue))
  ;[1 1 1] Debug Log Speed
  (Development|PrintString (Utilities|String|ToString(Float) (- (Utilities|Time|GetTimeSeconds) (Variables|Default|GetStartTime))) true true "(R=0.000000,G=0.660000,B=1.000000,A=1.000000)" 3e+01 "WingSuit_TimePassed")
  (Development|PrintString (Utilities|String|Append "Current Horizontal Speed: " (Utilities|String|ToString(Float) (Math|Vector|VectorLengthXY _returnvalue_2))) true true "(R=0.000000,G=0.660000,B=1.000000,A=1.000000)" 3e+01 "WingSuit_HorizontalVelocity")
  (Development|PrintString (Utilities|String|Append "Total Speed: " (Utilities|String|ToString(Float) (Math|Vector|VectorLength _returnvalue_2))) true true "(R=0.000000,G=0.660000,B=1.000000,A=1.000000)" 3e+01 "WingSuit_TotalVelocity")
  ;;[1 1 1] Manage speed history
  (Utilities|Array|Add (Variables|Default|GetHorizontalSpeedHistory) (Math|Vector|VectorLengthXY (Transformation|GetVelocity (CallFunction|GetAvatarasFortPlayerPawn))))
  (if (> (Utilities|Array|Length (Variables|Default|GetHorizontalSpeedHistory)) 600)
    (Utilities|Array|RemoveIndex (Variables|Default|GetHorizontalSpeedHistory)))
  (Default|PrintRunningAverageHorizontalSpeed))

(event Custom|TODOTick
  ;[1 1 1] REMOVE THIS WITH JEFF SENDING THE EVENTS
  (if (> (Math|Vector|VectorLength (Transformation|GetVelocity (CallFunction|GetAvatarasFortPlayerPawn))) (Variables|Default|GetVisualSpeed))
    (CallFunction|AddFastVisualEffects)
    (else
      (CallFunction|RemoveVisualEffects))))

(event Ability|EventOnEndAbility
  (params (bWasCancelled :type Boolean))
  (bind _local_player_controller (Variables|Default|GetLocal%20Player%20Controller))
  (|CallParentFunction false)
  (bind _mme (Variables|Default|GetMME)
    (:then
      (Class|FortMovementModeExtBlueprintLibrary|EndMovementModeExtensiononActor _mme (CallFunction|GetAvatarasFortPlayerPawn)))
    (:else
      _))
  (Ability|RemoveGameplayEffectFromOwnerWithHandle (Variables|Default|GetgameplayEffectForSmoke))
  (Ability|RemoveGameplayEffectFromOwnerWithHandle (Variables|Default|GetgameplayEffectForFlight))
  (Ability|RemoveGameplayCueFromOwner (Variables|Default|GetGameplay%20Cue%20Tag))
  (Camera|PopCameraMode)
  (Utilities|IsValid _local_player_controller
    (:Is%20Valid
      (Input|ResetInputRotationScale _local_player_controller))
    (:Is%20Not%20Valid
      _))
  (bind _asb_weap_wing_suit_physics_flying (Utilities|Casting|CastToB_Weap_WingSuit_PhysicsFlying (Weapon|GetCurrentWeapon (CallFunction|GetAvatarasFortPlayerPawn))))
  (Ability|HasAuthority
    (:True
      (Variables|/WingSuit/Gameplay/WingSuit_PhysicsFlying/B_Weap_WingSuit_PhysicsFlying.B_Weap_WingSuit_PhysicsFlying_C|SetSmokeScreenColor (Math|Random|RandomIntegerinRange 0 7) _asb_weap_wing_suit_physics_flying))
    (:False
      _))
  (if (ScalableFloat|GetValueasBool (Variables|Default|GetIsFreeDeploysEnabled))
    (if (Variables|/WingSuit/Gameplay/WingSuit_PhysicsFlying/B_Weap_WingSuit_PhysicsFlying.B_Weap_WingSuit_PhysicsFlying_C|GetFreeDeploy _asb_weap_wing_suit_physics_flying)
      (Variables|/WingSuit/Gameplay/WingSuit_PhysicsFlying/B_Weap_WingSuit_PhysicsFlying.B_Weap_WingSuit_PhysicsFlying_C|SetFreeDeploy false _asb_weap_wing_suit_physics_flying)
      (else
        (Ability|ApplyGameplayEffectToOwner "/WingSuit/Gameplay/WingSuit_PhysicsFlying/GE_WingSuit_PhysicsFlying_Cooldown.GE_WingSuit_PhysicsFlying_Cooldown_C")
        (Ability|ApplyGameplayEffectToOwner "/WingSuit/Gameplay/WingSuit_PhysicsFlying/GE_WingSuit_PhysicsFlying_Cooldown_Visuals.GE_WingSuit_PhysicsFlying_Cooldown_Visuals_C")))
    (bind _asb_weap_wing_suit_physics_flying_2 (Utilities|Casting|CastToB_Weap_WingSuit_PhysicsFlying (Weapon|GetCurrentWeapon (CallFunction|GetAvatarasFortPlayerPawn)))
      (:then
        (if (Variables|/WingSuit/Gameplay/WingSuit_PhysicsFlying/B_Weap_WingSuit_PhysicsFlying.B_Weap_WingSuit_PhysicsFlying_C|GetDeleteOnLand _asb_weap_wing_suit_physics_flying_2)
          (Inventory|RemoveItemfromInventoryOwner (Fortnite|GetFortPlayerControllerfromActor (CallFunction|GetAvatarasFortPlayerPawn)) "/WingSuit/Gameplay/WingSuit_PhysicsFlying/WID_Proto_WingSuit_FlyingPhysics.WID_Proto_WingSuit_FlyingPhysics")
          (else
            _))
        (Components|Activation|Deactivate (Variables|Default|GetCurrentVfx))
        (CallFunction|RemoveVisualEffects))
      (:CastFailed
        _))))

(event Custom|AddFastVisualEffects
  ;[1 1 1] Fast
  (if (not (Variables|Default|GetlensActive))
    (Variables|Default|SetlensActive true)
    (bind _returnvalue (Camera|AddGenericCameraLensEffect (Variables|/Script/Engine.PlayerController|GetPlayerCameraManager (Fortnite|GetFortPlayerControllerfromActor (CallFunction|GetAvatarasFortPlayerPawn))) "/WingSuit/VFX/BP_CameraLens_WingSuit_Sprint.BP_CameraLens_WingSuit_Sprint_C"))
    (Variables|Default|Setcamlenseffect _returnvalue)))

(event Custom|RemoveVisualEffects
  ;[1 1 1] Slow
  (if (Variables|Default|GetlensActive)
    (Variables|Default|SetlensActive false)
    (Camera|RemoveGenericCameraLensEffect (Variables|/Script/Engine.PlayerController|GetPlayerCameraManager (Fortnite|GetFortPlayerControllerfromActor (CallFunction|GetAvatarasFortPlayerPawn))) (Variables|Default|Getcamlenseffect))
    (Components|Activation|Deactivate (CameraLensEffect|GetPrimaryParticleComponent (Variables|Default|Getcamlenseffect)))))

(fn PrintRunningAverageHorizontalSpeed
  (if (Utilities|Array|IsNotEmpty (Variables|Default|GetHorizontalSpeedHistory))
    (for _array_element (Variables|Default|GetHorizontalSpeedHistory)
      (LocalVariables|SetRunningTotal (+ _array_element (LocalVariables|GetRunningTotal))))
    (Development|PrintString (Utilities|String|Append "Horizontal Average Speed: " (Utilities|String|ToString(Float) (/ (LocalVariables|GetRunningTotal) (Utilities|Array|Length (Variables|Default|GetHorizontalSpeedHistory))))) true true "(R=0.000000,G=0.660000,B=1.000000,A=1.000000)" 2e+01 "AverageSpeed")))
