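; Wingsuit self-launch gameplay ability that spends weapon ammo and throws the player upward, then activates the wingsuit flying movement extension on a timer.
; Ground launches wait for a montage notify while airborne ones proceed immediately, both sharing one launch path.
; That path keeps horizontal velocity, applies fall immunity, camera effects and cue, and capsule-traces to damage obstacles above.
; Helpers gate repeat jumps, flag the weapon's launch state and clear blocking tags on end.
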(event Custom|EndLoop
  ;[1 1 1] end anim loop
  (if (Variables|Default|GetbIsGroundJump)
    (Animation|Montage|MontageJumptoSection (Components|SkeletalMesh|GetAnimInstance (Variables|/Script/Engine.Character|GetMesh (CallFunction|GetAvatarasFortPlayerPawn))) "End" (Variables|Default|GetGroundLaunchMontage))
    (else
      (Animation|Montage|MontageJumptoSection (Components|SkeletalMesh|GetAnimInstance (Variables|/Script/Engine.Character|GetMesh (CallFunction|GetAvatarasFortPlayerPawn))) "End" (Variables|Default|GetFallingLaunchMontage)))))

(event Custom|MMEActivate
  (bind _returnvalue (CallFunction|GetAvatarasFortPlayerPawn))
  (if (GameplayTags|HasMatchingGameplayTag (CallFunction|GetAvatarasFortPlayerPawn) "(TagName=\"MovementMode.Flying.WingSuit\")")
    _
    (elif (ScalableFloat|GetValueasBool (Variables|Debug|GetUse%20Native%20MME?))
      (bind _returnvalue_2 (Class|FortMovementModeExtBlueprintLibrary|ActivateRegisteredMovementModeExtension "/WingSuit/Gameplay/WingSuit_PhysicsFlying/Native/FMM_ExtLogicWingSuit.FMM_ExtLogicWingSuit_C" _returnvalue))
      (Variables|Default|SetMME _returnvalue_2)
      (else
        (bind _returnvalue_3 (Class|FortMovementModeExtBlueprintLibrary|ActivateRegisteredMovementModeExtension 0 _returnvalue))
        (Variables|Default|SetMME _returnvalue_3))))
  (Ability|EndAbility))

(event Ability|EventActivateAbility
  (bind _returnvalue (CallFunction|GetAvatarasFortPlayerPawn))
  (bind _returnvalue_2 (Transformation|GetVelocity _returnvalue))
  (bind _x (.x _returnvalue_2))
  (bind _y (.y _returnvalue_2))
  (bind _returnvalue_3 (Math|Vector|MakeVector _x _y 4.5e+03))
  (bind _polarityweapon (Default|GetPolarityWeapon))
  (bind _returnvalue_4 (CallFunction|GetAvatarasFortPlayerPawn))
  (bind _returnvalue_5 (Weapon|GetCurrentWeapon _returnvalue_4))
  (bind _asfort_inventory_owner_interface (Utilities|Casting|CastToFortInventoryOwnerInterface (Pawn|GetController (CallFunction|GetAvatarasFortPlayerPawn))))
  (bind _returnvalue_6 (InventoryManagement|FindExistingItemforDefinition _asfort_inventory_owner_interface "/WingSuit/Gameplay/WingSuit_PhysicsFlying/WID_Proto_WingSuit_FlyingPhysics.WID_Proto_WingSuit_FlyingPhysics"))
  (bind _asfort_world_item (Utilities|Casting|CastToFortWorldItem _returnvalue_6))
  (bind _asb_weap_wing_suit_physics_flying (Utilities|Casting|CastToB_Weap_WingSuit_PhysicsFlying _returnvalue_5))
  (Variables|/WingSuit/Gameplay/WingSuit_PhysicsFlying/B_Weap_WingSuit_PhysicsFlying.B_Weap_WingSuit_PhysicsFlying_C|SetDeleteOnLand (<= (Item|GetTotalAmmo _asfort_world_item) 1) _asb_weap_wing_suit_physics_flying)
  (FortWeapon|Ammo|ConsumeAmmo _returnvalue_5 1 true)
  (bind _asfort_pawn (Utilities|Casting|CastToFortPawn (Ability|GetAvatarActorfromActorInfo)))
  (Variables|Default|SetPlayerPawn _asfort_pawn)
  (bind _bcanbassjump (CallFunction|CanBassJump))
  (if _bcanbassjump
    (Ability|SendGameplayEvent "(TagName=\"Event.Weapons.Fire\")")
    (Ability|SendGameplayEvent "(TagName=\"WingSuit.UseAbility\")")
    (Ability|GameplayEffect|AddLooseGameplayTags (Variables|Default|GetPlayerPawn) (Variables|Default|GetTagsBlockBassWave))
    (Camera|PushCameraMode "/GravWellGameplay/Assets/Camera/CameraMode_BassWave_SelfLaunch.CameraMode_BassWave_SelfLaunch_C")
    (CallFunction|HasAbilityStarted true)
    (if (GameplayTags|HasMatchingGameplayTag (Variables|Default|GetPlayerPawn) "(TagName=\"MovementMode.Falling\")")
      ;[1 1 1] If player is airborne, don't wait for the trigger
      (Ability|Tasks|PlayMontageandWaitforTrigger "None" (Variables|Default|GetFallingLaunchMontage)
        (:then
          (Variables|Default|SetbIsGroundJump false)
          (CallFunction|PauseUntilNextJump)
          (Utilities|Time|SetTimerbyEvent (delegate Custom|End%20Camera) 2.0 :!label _label)
          ;[1 1 1] end anim loop
          (Utilities|Time|SetTimerbyEvent (delegate Custom|EndLoop) 0.65)
          (Game|Feedback|ClientStartCameraShake :self (Fortnite|GetFortPlayerControllerfromActor (Variables|Default|GetPlayerPawn)) :Shake (Variables|Default|GetBassJumpCamShake) :Scale 1.0 :PlaySpace "CameraLocal")
          (Ability|ApplyGameplayEffectToOwner (Variables|Default|GetFallImmunityGE))
          (Utilities|IsValid _polarityweapon
            (:Is%20Valid
              (Class|BWeapWingSuitPhysicsFlying|TryUseDurability _polarityweapon (ScalableFloat|GetValueatLevel (Variables|Default|GetHF_AbilityDurabilityUseMultiplier))))
            (:Is%20Not%20Valid
              _))
          (Ability|CommitAbility)
          (Ability|ExecuteGameplayCueOnOwner (Variables|Default|GetGCBassJumpLaunch))
          (CallFunction|BassJumpBlastThrough)
          (Utilities|Time|SetTimerbyEvent (delegate Custom|MMEActivate) 1.5)
          (Ability|HasAuthority
            (:True
              (Pawn|Character|LaunchCharacterJump _returnvalue _returnvalue_3 true true false true)
              ;[1 1 1] Wait until the player has landed before ending the ability- maybe needs some work to make it more robust?
              (bind _asfort_pawn_2 (Utilities|Casting|CastToFortPawn (Ability|GetAvatarActorfromActorInfo)))
              (Ability|Tasks|WaitGameplayTagAddToActor _asfort_pawn_2 "(TagName=\"MovementMode.Falling\")"
                (:Added
                  (Ability|Tasks|WaitGameplayTagRemoveFromActor _asfort_pawn_2 "(TagName=\"MovementMode.Falling\")" true)
                  (Utilities|Time|SetTimerbyEvent (delegate Custom|EndAbilityTimeout) 2e+01))))
            (:False
              (Pawn|Character|LaunchCharacterJump _returnvalue _returnvalue_3 true true false true))))
        (:Cancelled
          _)
        (:Completed
          _))
      (else
        (Ability|Tasks|PlayMontageAndWaitNotify "None" (Variables|Default|GetGroundLaunchMontage) 1.0 "Start"
          (:then
            (Variables|Default|SetbIsGroundJump true)
            (CallFunction|PauseUntilNextJump))
          (:OnCompleted
            _)
          (:OnInterrupted
            _)
          (:OnCancelled
            _)
          (:OnNotifyBegin
            (goto _label)))))
    (CallFunction|HasAbilityStarted false)))

(event Custom|EndAbilityTimeout
  ;[1 1 1] Wait until the player has landed before ending the ability- maybe needs some work to make it more robust?
  (Ability|EndAbility))

(event Custom|End%20Camera
  (Camera|PopCameraMode))

(event Custom|RemoveLaunchingTags
  (Ability|GameplayEffect|RemoveLooseGameplayTags (Variables|Default|GetPlayerPawn) (Variables|Default|GetTagsBlockBassWave) true))

(event Custom|PauseUntilNextJump
  ;[1 1 1] After Bass Jumping, allow next Bass Jump after X duration (separate duration for first jump to allow for later ability application)
  (if (Variables|Default|GetbIsGroundJump)
    (Utilities|Time|SetTimerbyEvent (delegate Custom|AllowNextJump) (Default|ClampMinFloat (ScalableFloat|GetValueatLevel (Variables|Default|GetGround%20TIme%20Until%20Allow%20Next%20Jump)) (ScalableFloat|GetValueasDouble (Variables|Default|GetTime%20Until%20Allow%20Next%20Jump%20Minimum))))
    (else
      (Utilities|Time|SetTimerbyEvent (delegate Custom|AllowNextJump) (Default|ClampMinFloat (ScalableFloat|GetValueatLevel (Variables|Default|GetAerial%20Time%20Until%20Allow%20Next%20Jump)) (ScalableFloat|GetValueasDouble (Variables|Default|GetTime%20Until%20Allow%20Next%20Jump%20Minimum)))))))

(event Custom|AllowNextJump)

(event Ability|EventOnEndAbility
  (params (bWasCancelled :type Boolean))
  (CallFunction|HasAbilityStarted false)
  (CallFunction|RemoveLaunchingTags)
  (if (Variables|Default|GetdeleteLater)
    (Inventory|RemoveItemfromInventoryOwner (Fortnite|GetFortPlayerControllerfromActor (CallFunction|GetAvatarasFortPlayerPawn)) "/WingSuit/Gameplay/WingSuit_PhysicsFlying/WID_Proto_WingSuit_FlyingPhysics.WID_Proto_WingSuit_FlyingPhysics")))

(fn BassJumpBlastThrough
  (bind (_returnvalue_x _returnvalue_y) (Math|Vector|BreakVector (Transformation|GetActorLocation (Variables|Default|GetPlayerPawn))))
  (bind (_returnvalue_x_2 _returnvalue_y_2 _returnvalue_z) (Math|Vector|BreakVector (Transformation|GetActorLocation (Variables|Default|GetPlayerPawn))))
  (if (> (ScalableFloat|GetValueatLevel (Variables|Default|GetBassJumpShouldDestroyAbove)) 0.0)
    ;[1 1 1] Currently just applies vertically, investigate matching it to the player's path
    (bind _outhits (Collision|MultiCapsuleTraceForObjects :Start (Math|Vector|MakeVector _returnvalue_x_2 _returnvalue_y_2 (+ _returnvalue_z (ScalableFloat|GetValueatLevel (Variables|Default|GetBassJumpDestroyDistanceZOffset)))) :End (Math|Vector|MakeVector _returnvalue_x _returnvalue_y (+ _returnvalue_z (ScalableFloat|GetValueatLevel (Variables|Default|GetBassJumpDestroyDistanceZOffset)))) :Radius 1.6e+02 :HalfHeight 3.5e+02 :ObjectTypes (Variables|Default|GetObjectTypesToDestroy) :bTraceComplex false :DrawDebugType "ForDuration" :bIgnoreSelf true :TraceColor "(R=1.000000,G=0.000000,B=0.000000,A=1.000000)" :TraceHitColor "(R=0.000000,G=1.000000,B=0.000000,A=1.000000)" :DrawTime 8.0))
    (for _array_element _outhits
      (if (Variables|/Script/Engine.Actor|GetbCanBeDamaged (.HitActor (Collision|BreakHitResult _array_element)))
        (bind _bdidactorhavetags (Class|AthenaQuestFunctionLibrary|AthenaCheckActorHasTags (.HitActor (Collision|BreakHitResult _array_element)) "(GameplayTags=((TagName=\"Deimos.Spawner\")))"))
        (if (not _bdidactorhavetags)
          (Utilities|Casting|CastToBuildingTrap (.HitActor (Collision|BreakHitResult _array_element))
            (:CastFailed
              (GameplayEffects|ApplyGameplayEffectToTarget :self (Ability|GetAbilitySystemComponent (Variables|Default|GetPlayerPawn)) :GameplayEffectClass "/Game/Athena/Items/Consumables/ShockwaveGrenade/GE_ShockGrenade_Damage.GE_ShockGrenade_Damage_C" :Target (Ability|GetAbilitySystemComponent (.HitActor (Collision|BreakHitResult _array_element))) :Level 0.0))))))))

(fn CanBassJump
  ;[1 1 1] If player can multi-jump skip other checks
  (if (ScalableFloat|GetValueasBool (Variables|Default|GetbAllowMultiBassJump))
    (return true)
    (else
      (return (not (GameplayTags|HasMatchingGameplayTag (Variables|Default|GetPlayerPawn) (Variables|Default|GetTagPlayerBassJumpActive)))))))

(fn GetPolarityWeapon
  (bind _asb_weap_wing_suit_physics_flying (Utilities|Casting|CastToB_Weap_WingSuit_PhysicsFlying (Ability|GetCurrentSourceWeapon))
    (:then
      (return _asb_weap_wing_suit_physics_flying))
    (:CastFailed
      (return 0))))

(fn HasAbilityStarted
  (params (AbilityStart :type Boolean))
  (bind _asb_weap_wing_suit_physics_flying (Utilities|Casting|CastToB_Weap_WingSuit_PhysicsFlying (Weapon|GetCurrentWeapon (Variables|Default|GetPlayerPawn))))
  (Variables|/WingSuit/Gameplay/WingSuit_PhysicsFlying/B_Weap_WingSuit_PhysicsFlying.B_Weap_WingSuit_PhysicsFlying_C|SetLaunchActivated AbilityStart _asb_weap_wing_suit_physics_flying))

(fn CanActivateAbility
  (params (ActorInfo :type Gameplay%20Ability%20Actor%20Info%20Structure) (Handle :type Gameplay%20Ability%20Spec%20Handle%20Structure))
  (return 0 (>= (ScalableFloat|GetValueatLevel (Variables|Default|GetHF_CanActivateAbility)) 1.0)))
