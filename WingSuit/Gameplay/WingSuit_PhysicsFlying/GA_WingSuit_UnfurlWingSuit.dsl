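; Gameplay ability that unfurls the wingsuit on the player pawn for as long as wingsuit flight lasts.
; On activation it gives the configured item, applies the transformation visuals effect and adds the wingsuit-on loose tag.
; It then waits for removal of the MovementMode.Flying.WingSuit tag to end itself.
; Ending removes the item and tag and applies the same visuals effect again, covering both transitions.
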(event Ability|EventActivateAbility
  (Inventory|GiveItemtoInventoryOwner :InventoryOwner (Fortnite|GetFortPlayerControllerfromActor (CallFunction|GetAvatarasFortPlayerPawn)) :ItemDefinition (Variables|Default|GetItemToGive) :NumberToGive 1 :bNotifyPlayer true :ItemLevel -1 :PickupInstigatorHandle 1 :bUseItemPickupAnalyticEvent false :WeaponAmmoOverride -1)
  (GameplayEffects|ApplyGameplayEffectToSelf :self (Ability|GetAbilitySystemComponent (CallFunction|GetAvatarasFortPlayerPawn)) :GameplayEffectClass "/WingSuit/Gameplay/WingSuit_PhysicsFlying/GE_TransformationVisuals.GE_TransformationVisuals_C" :Level 0.0)
  (Ability|GameplayEffect|AddLooseGameplayTags (CallFunction|GetAvatarasFortPlayerPawn) (Variables|Default|GetWingsuitOnTag) true)
  (Ability|Tasks|WaitGameplayTagRemove "(TagName=\"MovementMode.Flying.WingSuit\")" 0 true
    (:Removed
      (Ability|EndAbility))))

(event Ability|EventOnEndAbility
  (params (bWasCancelled :type Boolean))
  (Inventory|RemoveItemfromInventoryOwner (Fortnite|GetFortPlayerControllerfromActor (CallFunction|GetAvatarasFortPlayerPawn)) (Variables|Default|GetItemToGive))
  (GameplayEffects|ApplyGameplayEffectToSelf :self (Ability|GetAbilitySystemComponent (CallFunction|GetAvatarasFortPlayerPawn)) :GameplayEffectClass "/WingSuit/Gameplay/WingSuit_PhysicsFlying/GE_TransformationVisuals.GE_TransformationVisuals_C" :Level 0.0)
  (Ability|GameplayEffect|RemoveLooseGameplayTags (CallFunction|GetAvatarasFortPlayerPawn) (Variables|Default|GetWingsuitOnTag) true))
